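# Material function for dealership FX that outputs two paint variants, a textured diffuse version and a mask-only particle color version.
# An RGB selector vector is dotted with the opacity texture input to pick channels as a mask, output directly as ParticleColorVersion.
# DiffuseVersion multiplies that mask by the diffuse texture's RGB, sampled with UVs rotated by CustomRotator through a scalar angle parameter.
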MaterialFunction /DealershipSharedLibrary/FX/Materials/MaterialFunctions/MF_VCFX_PaintedVersions
graph {
 properties {previewBlendMode=BLEND_Translucent}
 nodes [
  n1 = MaterialFunctionCall(UVs=TextureCoordinate(), "Rotation Angle (0-1)"=ScalarParameter() {parameterName="Rotation Angle (0-1) (S)"}) {materialFunction=&CustomRotator}
  n2 = FunctionInput(Preview=TextureObjectParameter(UVs=@n1.Rotated Values) {parameterName=Diffuse_Texture}) {bUsePreviewValueAsDefault=true, inputName=Diffuse_Texture, inputType=FunctionInput_Texture2D}
  n3 = DotProduct(
   A=FunctionInput() {bUsePreviewValueAsDefault=true, inputName=OpacityTexture}
   B=FunctionInput(Preview=VectorParameter() {parameterName=RGBSelectorTextureOpacity, defaultValue=<rgba 0 0 0 1>}.RGB) {bUsePreviewValueAsDefault=true, inputName=RGB Selector}
  )
  FunctionOutput(None=Multiply(A=TextureSample(UVs=@n1.Rotated Values, Tex=@n2) {texture=/DealershipSharedLibrary/FX/Textures/Diffuse/T_VCVFX_HeadStarBG02_D}.RGB, B=@n3)) {outputName=DiffuseVersion, sortPriority=1}
  FunctionOutput(None=@n3) {outputName=ParticleColorVersion}
 ]
}
