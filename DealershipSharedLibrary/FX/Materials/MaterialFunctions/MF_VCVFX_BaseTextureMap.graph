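# VFX material function that samples a texture through a configurable UV chain and outputs RGB and individual RGBA channels.
# Saturated UVs blend toward flipbook UVs (MF_SubUV or inline math, chosen by NewSubUVSystem), then are quantized for pixelation.
# They are then offset or mirrored, tiled, panned, scaled, rotated and optionally distorted by a static switch.
# Tiling, offset, panning speed and rotation center each feed one component to both axes, despite two-axis input layouts.
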MaterialFunction /DealershipSharedLibrary/FX/Materials/MaterialFunctions/MF_VCVFX_BaseTextureMap
graph {nodes [
  n1 = TextureCoordinate()
  n2 = ScalarParameter() {group=Flipbook, parameterName=FrameRate, sortPriority=2}
  n3 = ScalarParameter() {defaultValue=1, group=Flipbook, parameterName=Columns, sortPriority=0}
  n4 = ScalarParameter() {defaultValue=1, group=Flipbook, parameterName=Rows, sortPriority=1}
  n5 = MaterialFunctionCall!(U_Images=@n3, V_Images=@n4, SubUV Texture RGB, Time Speed=@n2, Manual Time Control, Time, UVs) {materialFunction=/Game/Effects/Fort_Effects/Materials/_Functions/MF_SubUV}
  n6 = Multiply(A=Time() {period=1}, B=@n2)
  n7 = Floor(None=@n6)
  n8 = Floor(None=Divide(A=@n6, B=@n3))
  n9 = Multiply(A=Divide(A=Constant() {r=1}, B=AppendVector(A=@n3, B=@n4)), B=Add(A=@n1, B=AppendVector(A=@n7, B=@n8)))
  n10 = LinearInterpolate(
   A=Saturate(None=@n1)
   B=If(A=ScalarParameter() {parameterName=NewSubUVSystem}, "A > B"=@n5.UVs, "A == B"=@n5.UVs, "A < B"=Frac(None=@n9)) {constB=0.5}
   Alpha=FunctionInput(Preview=ScalarParameter() {parameterName=Spritesheet?}) {bUsePreviewValueAsDefault=true, inputName=spritesheet?, inputType=FunctionInput_Scalar}
  )
  n11 = FunctionInput(Preview=ScalarParameter() {defaultValue=1024, parameterName=PixelBits}) {bUsePreviewValueAsDefault=true, inputName=Pixelbits, inputType=FunctionInput_Scalar}
  n12 = Divide(A=Floor(None=Multiply(A=@n10, B=@n11)), B=@n11)
  n13 = FunctionInput(Preview=Constant() {r=1}) {bUsePreviewValueAsDefault=true, inputName=Offset/Mirror X 1stTexture, inputType=FunctionInput_Scalar}
  n14 = FunctionInput(Preview=Constant() {r=1}) {bUsePreviewValueAsDefault=true, inputName=Offset/Mirror Y 1stTexture, inputType=FunctionInput_Scalar}
  n15 = Add(A=@n12, B=AppendVector(A=@n13, B=@n14))
  n16 = Multiply(A=@n12, B=AppendVector(A=@n13, B=@n14))
  n17 = FunctionInput(Preview=Constant() {r=1}) {
   bUsePreviewValueAsDefault=true
   inputName="Offset = 0// Mirror = 1 ForSocket 1stTextureTexture"
   inputType=FunctionInput_Scalar
   previewValue=<1 1 1 1>
  }
  n18 = FunctionInput() {
   bUsePreviewValueAsDefault=true
   desc=RG controls Tiling. BA controls Offset
   inputName=Tiling N Offset
   inputType=FunctionInput_Vector4
   sortPriority=1
   previewValue=<1 1 0 0>
  }
  n19 = MaterialFunctionCall(Float4=@n18) {materialFunction=&BreakOutFloat4Components}
  n20 = Add(A=Multiply(A=LinearInterpolate(A=@n15, B=@n16, Alpha=@n17), B=AppendVector(A=@n19.R, B=@n19.R)), B=AppendVector(A=@n19.B, B=@n19.B))
  n21 = FunctionInput() {
   bUsePreviewValueAsDefault=true
   desc=RG controls XY speed respectively. B controls panning time scale.
   inputName=Panning
   sortPriority=2
   previewValue=<0 0 1 0>
  }
  n22 = MaterialFunctionCall(Float3=@n21) {materialFunction=&BreakOutFloat3Components}
  n23 = Multiply(A=Time(), B=@n22.B)
  n24 = AppendVector(A=@n22.R, B=@n22.R)
  n25 = FunctionInput() {bUsePreviewValueAsDefault=true, inputName=Scale, inputType=FunctionInput_Vector2, sortPriority=3, previewValue=<1 1 0 1>}
  n26 = FunctionInput() {
   bUsePreviewValueAsDefault=true
   desc="RG controls rotation center point. B controls normalized rotation angle (0 - 1)."
   inputName=Rotation Point N Angle
   sortPriority=4
   previewValue=<0.5 0.5 0 0>
  }
  n27 = MaterialFunctionCall(Float3=@n26) {materialFunction=&BreakOutFloat3Components}
  n28 = MaterialFunctionCall(
   UVs=MaterialFunctionCall(UVs=Panner(Coordinate=@n20, Time=@n23, Speed=@n24), Texture Scale=@n25) {materialFunction=&ScaleUVsByCenter}.UVs
   Rotation Center=AppendVector(A=@n27.R, B=@n27.R)
   "Rotation Angle (0-1)"=@n27.B
  ) {materialFunction=&CustomRotator}
  n29 = FunctionInput() {
   bUsePreviewValueAsDefault=true
   desc="Distortion layer data or channel to be used on this texture. \"Use distortion?\" must be true/enable."
   description="Distortion layer data or channel to be used on this texture. \"Use distortion?\" must be true/enable."
   inputName=Distortion Layer
   sortPriority=6
  }
  n30 = FunctionInput(Preview=StaticBool()) {
   bUsePreviewValueAsDefault=true
   desc=Set true to allow this texture config to use a distortion layer.
   description=Set true to allow this texture config to use a distortion layer.
   inputName=Use distortion?
   inputType=FunctionInput_StaticBool
   sortPriority=5
  }
  n31 = TextureSample(
   UVs=StaticSwitch(True=Add(A=@n28.Rotated Values, B=@n29), False=@n28.Rotated Values, Value=@n30)
   Tex=FunctionInput(Preview=TextureObject()) {bUsePreviewValueAsDefault=true, inputName=Texture, inputType=FunctionInput_Texture2D}
  ) {texture=/Engine/EngineResources/DefaultTexture}
  FunctionOutput(None=@n31.RGB) {outputName=RGB}
  FunctionOutput(None=@n31.R) {outputName=R, sortPriority=1}
  FunctionOutput(None=@n31.G) {outputName=G, sortPriority=2}
  FunctionOutput(None=@n31.B) {outputName=B, sortPriority=3}
  FunctionOutput(None=@n31.A) {outputName=A, sortPriority=4}
 ]}
