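# Translucent unlit master VFX material, usable on Niagara ribbons, that blends two textures with particle color for emissive and opacity.
# Each layer runs through shared texture-mapping and PaintedVersions functions, with rotatable procedural gradient masks shaping its alpha.
# Layers fuse by multiply or add, separately for color and opacity, then feed emissive-adjustment and depth-fade functions.
# Scalar lerps select optional world position offset, either sine waves or twisting helix, with dynamic parameters for per-particle control.
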Material /DealershipSharedLibrary/FX/Materials/MasterMaterials/General/MM_2_TexturesPaintedVersions
graph {
 properties {bUsedWithNiagaraRibbons=true, blendMode=BLEND_Translucent, shadingModel=MSM_Unlit, usedShadingModels=Unlit}
 nodes [
  n1 = Constant()
  n2 = ComponentMask(None=TextureCoordinate() {coordinateIndex=1}) {r=true}
  n3 = ScalarParameter() {defaultValue=1, group=WOPWaves, parameterName=CountWaves2}
  n4 = ScalarParameter() {defaultValue=1, group=WPO_Helix, parameterName=Speed}
  n5 = OneMinus(None=Time())
  n6 = Sine(None=Add(A=Divide(A=@n2, B=@n3), B=Multiply(A=@n4, B=@n5)))
  n7 = ScalarParameter() {defaultValue=7, group=WOPWaves, parameterName=Strenght}
  n8 = Power(Base=ComponentMask(None=TextureCoordinate()) {r=true}, Exp=ScalarParameter() {defaultValue=1, group=WOPWaves, parameterName=PowerMiddle2})
  n9 = ScalarParameter() {defaultValue=0.5, group=WOPWaves, parameterName=MultPower2}
  n10 = AppendVector(A=AppendVector(A=@n1, B=@n1), B=Multiply(A=Multiply(A=@n6, B=@n7), B=Multiply(A=@n8, B=@n9)))
  n11 = TextureCoordinate()
  n12 = MaterialFunctionCall(
   In=Multiply(A=ComponentMask(None=@n11) {r=true}, B=ScalarParameter() {defaultValue=1, group=WOPWaves, parameterName=LockGradientValueWOP}) {constB=2}
   Contrast=ScalarParameter() {defaultValue=50, group=WOPWaves, parameterName=Contrast_MaskWOIP}
  ) {materialFunction=&CheapContrast}
  n13 = DynamicParameter() {parameterIndex=3, defaultValue=<rgba 50 1 1 1>, paramNames [WOPUP,Param2,Param3,Param4]}
  n14 = Multiply(A=Constant3Vector() {constant=<rgba 0 0 1 0>}, B=@n13.WOPUP) {constB=50}
  n15 = Add(A=TextureCoordinate(), B=AppendVector(A=Constant() {r=1.082}, B=Constant() {r=1}))
  n16 = MaterialFunctionCall(
   In=Sine(None=MaterialFunctionCall(UV Channel=@n15) {materialFunction=&LinearGradient}.UGradient)
   Contrast=@n13.Param2
  ) {materialFunction=&CheapContrast}
  n17 = Add(
   A=Multiply(A=ScalarParameter() {group=WOPWaves, parameterName=PowerWOP}, B=Multiply(A=@n10, B=@n12.Result))
   B=LinearInterpolate(B=Multiply(A=@n14, B=@n16.Result) {constA=1}, Alpha=ScalarParameter() {group=WOPWaves, parameterName=UpLock?})
  )
  n18 = ScalarParameter() {defaultValue=10, group=WPO_Helix, parameterName=Power_Helix}
  n19 = Multiply(A=OneMinus(None=Time()), B=ScalarParameter() {defaultValue=1, group=WPO_Helix, parameterName=Speed})
  n20 = ScalarParameter() {defaultValue=1, group=WPO_Helix, parameterName=WPO_TileAmt_Helix}
  n21 = TextureCoordinate()
  n22 = Divide(A=ComponentMask(None=Multiply(A=@n20, B=@n21)) {r=true}, B=ScalarParameter() {defaultValue=1, group=WPO_Helix, parameterName=tile_helix})
  n23 = Add(A=@n19, B=@n22)
  n24 = Add(A=@n22, B=@n19)
  n25 = AppendVector(A=AppendVector(A=Cosine(None=@n23) {period=0.5}, B=Sine(None=@n24)), B=Constant())
  n26 = Sine(None=@n23) {period=0.75}
  n27 = Cosine(None=@n24)
  n28 = AppendVector(A=AppendVector(A=OneMinus(None=@n26), B=OneMinus(None=@n27)), B=Constant())
  MaterialExpressionConstantClamp_1 = Saturate(None=LinearInterpolate(A=Sine(None=@n24) {period=2}, B=Cosine(None=@n23) {period=2.5}, Alpha=Constant() {r=0.5}))
  n29 = OneMinus(None=@n23)
  n30 = OneMinus(None=@n24)
  n31 = AppendVector(A=AppendVector(A=Cosine(None=@n29) {period=0.5}, B=Sine(None=@n30)), B=Constant())
  n32 = Sine(None=@n29) {period=0.75}
  n33 = Cosine(None=@n30)
  n34 = AppendVector(A=AppendVector(A=OneMinus(None=@n32), B=OneMinus(None=@n33)), B=Constant())
  MaterialExpressionConstantClamp_2 = Saturate(None=LinearInterpolate(A=Sine(None=@n30) {period=2}, B=Cosine(None=@n29) {period=2.5}, Alpha=Constant() {r=0.5}))
  n35 = Multiply(
   A=LinearInterpolate(
    A=LinearInterpolate(A=@n25, B=@n28, Alpha=@MaterialExpressionConstantClamp_1)
    B=LinearInterpolate(A=@n31, B=@n34, Alpha=@MaterialExpressionConstantClamp_2)
    Alpha=ScalarParameter() {group=WPO_Helix, parameterName=Mirror}
   )
   B=ScalarParameter() {defaultValue=1, group=WPO_Helix, parameterName=Intensity}
  )
  n36 = ScalarParameter() {defaultValue=1, group=WPO_Helix, parameterName=Amount}
  n37 = ConstantBiasScale(None=Transform(None=Multiply(A=@n35, B=@n36))) {scale=0.25}
  n38 = ComponentMask(None=TextureCoordinate()) {r=true}
  n39 = OneMinus(None=@n38)
  MaterialExpressionConstantClamp_0 = Saturate(None=Multiply(A=Multiply(A=@n39, B=@n38), B=Constant() {r=4}))
  n40 = Time()
  n41 = Add(
   A=ComponentMask(None=TextureCoordinate() {coordinateIndex=1}) {r=true}
   B=Multiply(A=ScalarParameter() {defaultValue=2, group=WPO_Helix, parameterName=Twist}, B=OneMinus(None=@n40))
  )
  n42 = OneMinus(None=@n41)
  n43 = Sine(None=@n42) {period=2}
  n44 = AppendVector(A=Cosine(None=@n42) {period=2}, B=@n43)
  n45 = Sine(None=@n41) {period=2}
  n46 = AppendVector(A=Cosine(None=@n41) {period=2}, B=@n45)
  n47 = Transform(None=LinearInterpolate(A=AppendVector(A=@n44, B=@n43), B=AppendVector(A=@n46, B=@n45), Alpha=ScalarParameter() {group=WPO_Helix, parameterName=Mirror}))
  n48 = Constant() {r=20}
  n49 = ScalarParameter() {defaultValue=2, group=WPO_Helix, parameterName=Pinch}
  n50 = ComponentMask(None=Panner(Coordinate=TextureCoordinate() {coordinateIndex=1, uTiling=0.8}, Time=Constant() {r=1}) {speedX=0.2}) {r=true}
  MaterialExpressionConstantClamp_3 = Saturate(None=ConstantBiasScale(None=Multiply(A=@n49, B=@n50)) {bias=0.4, scale=0.35})
  n51 = ComponentMask(None=Subtract(A=TextureCoordinate(), B=Constant() {r=0.5})) {g=true}
  n52 = Add(
   A=Multiply(A=Constant() {r=1}, B=Multiply(A=@n37, B=@MaterialExpressionConstantClamp_0))
   B=Multiply(A=Multiply(A=@n47, B=@n48), B=Multiply(A=@MaterialExpressionConstantClamp_3, B=@n51))
  )
  LinearInterpolate(
   B=LinearInterpolate(
    A=Transform(None=@n17) {transformSourceType=TRANSFORMSOURCE_Local}
    B=Multiply(A=@n18, B=@n52)
    Alpha=ScalarParameter() {parameterName="Waves = 0 // Helix = 1"}
   )
   Alpha=ScalarParameter() {parameterName="1._WOP? False= 0 True =1", sortPriority=1}
  )
  n53 = ParticleColor()
  n54 = DynamicParameter() {
   parameterIndex=2
   paramNames [Offset/Mirror X 1stTexture, Offset/Mirror Y 1stTexture, Offset/Mirror X 2ndTexture, Offset/Mirror Y 2ndTexture]
  }
  n55 = VectorParameter() {group=OpacityTexture_Main, parameterName=MainTexture_Tiling N Offset, defaultValue=<rgba 1 1 1 1>}
  n56 = VectorParameter() {group=OpacityTexture_Main, parameterName="MainTexture_Panning R= X / G = Y / B = Time", defaultValue=<rgba 0 0 1 0>}
  n57 = ScalarParameter() {defaultValue=1, group=OpacityTexture_Main, parameterName="ScaleMainTexture = X"}
  n58 = ScalarParameter() {defaultValue=1, group=OpacityTexture_Main, parameterName="ScaleMainTexture = Y"}
  n59 = VectorParameter() {group=OpacityTexture_Main, parameterName="MainTexture_RotationCenter R= X G= Y // B=RotationAngle", defaultValue=<rgba 0.5 0.5 0 0>}
  n60 = Multiply(
   A=ScalarParameter() {defaultValue=1, group=OpacityTexture_Main, parameterName=MainTexture_HardnessTexture}
   B=MaterialFunctionCall!(
    spritesheet?
    "Offset = 0// Mirror = 1 ForSocket 1stTextureTexture"=ScalarParameter() {defaultValue=1, group=OpacityTexture_Main, parameterName="Offset = 0// Mirror = 1 ForSocket 1stTextureTexture"}
    Texture=TextureObjectParameter() {group=OpacityTexture_Main, parameterName=MainTexture_OpacityTexture}
    Offset/Mirror Y 1stTexture=@n54.Offset/Mirror Y 1stTexture
    Pixelbits
    Offset/Mirror X 1stTexture=@n54.Offset/Mirror Y 1stTexture
    Tiling N Offset=MaterialFunctionCall(X=@n55.R, Y=@n55.R, Z=@n55.R, A=@n55.R) {materialFunction=&MakeFloat4}.Result
    Panning=@n56.RGB
    Scale=AppendVector(A=@n57, B=@n58)
    Rotation Point N Angle=@n59.RGB
    Use distortion?=StaticBoolParameter() {group=OpacityTexture_Main, parameterName=MainTexture_Use distortion?}
    Distortion Layer
   ) {materialFunction=/DealershipSharedLibrary/FX/Materials/MaterialFunctions/MF_VCVFX_BaseTextureMap}.RGB
  )
  n61 = ScalarParameter() {defaultValue=1, group=OpacityTexture_Main, parameterName=MainTexture_Contrast}
  n62 = MaterialFunctionCall!(
   OpacityTexture=Multiply(A=Power(Base=@n60, Exp=@n61), B=ScalarParameter() {defaultValue=1, group=OpacityTexture_Main, parameterName=MainTexture_Emissive})
   RGB Selector=VectorParameter() {group=OpacityTexture_Main, parameterName=RGBSelectorTextureOpacity, defaultValue=<rgba 1 1 1 1>}.RGB
   Diffuse_Texture=TextureObjectParameter() {group=OpacityTexture_Main, parameterName=MainTexture_DiffuseTexture}
  ) {materialFunction=/DealershipSharedLibrary/FX/Materials/MaterialFunctions/MF_VCFX_PaintedVersions}
  n63 = LinearInterpolate(
   A=@n62.ParticleColorVersion
   B=@n62.ParticleColorVersion
   Alpha=ScalarParameter() {group=OpacityTexture_Main, parameterName="MainTexture_ParticleColor  //DiffuseVersion"}
  )
  n64 = VectorParameter() {
   desc="RG controls normalized rotation angle (0 - 1) for Primary and Secondary mask respectively."
   group=Masks
   parameterName=Masks Rotation Angle
   sortPriority=8
   defaultValue=<rgba 0 0 0 1>
  }
  n65 = MaterialFunctionCall(UVs=TextureCoordinate(), "Rotation Angle (0-1)"=@n64.R) {materialFunction=&CustomRotator}
  n66 = ComponentMask(None=@n65.Rotated Values) {r=true}
  n67 = Sine(None=@n66) {period=2}
  n68 = OneMinus(None=@n67)
  n69 = Multiply(A=@n67, B=Sine(None=ComponentMask(None=@n65.Rotated Values) {g=true}) {period=2})
  n70 = VectorParameter() {
   desc=R controls Input Low. G controls Input High. B controls Target Low. A controls Target High.
   group=Masks
   parameterName=Primary Mask Set 1 - Remap
   sortPriority=1
   defaultValue=<rgba 0 1 0 1>
  }
  n71 = MaterialFunctionCall(
   Input=StaticComponentMaskParameter(None=MaterialFunctionCall(X=@n66, Y=@n67, Z=@n68, A=@n69) {materialFunction=&MakeFloat4}.Result) {defaultR=true, group=Masks, parameterName="Primary Mask Set 1 -  Selector", sortPriority=0}
   Input Low=@n70.R
   Input High=@n70.R
   Target Low=@n70.R
   Target High=@n70.R
  ) {materialFunction=&RemapValueRange}
  n72 = ScalarParameter() {group=Masks, parameterName=MainTexture_SimpleGradient//MirrorGradient}
  n73 = ConstantBiasScale(None=@n65.Rotated Values) {bias=-0.5, scale=2}
  n74 = Abs(None=ComponentMask(None=@n73) {r=true})
  n75 = OneMinus(None=ComponentMask(None=@n73) {g=true})
  n76 = Multiply(A=@n74, B=@n75)
  n77 = Multiply(A=OneMinus(None=@n74), B=@n75)
  n78 = VectorParameter() {
   desc=R controls Input Low. G controls Input High. B controls Target Low. A controls Target High.
   group=Masks
   parameterName=Primary Mask Set 2 - Remap
   sortPriority=3
   defaultValue=<rgba 0 1 0 1>
  }
  n79 = MaterialFunctionCall(
   Input=StaticComponentMaskParameter(None=MaterialFunctionCall(X=@n76, Y=@n77) {materialFunction=&MakeFloat2}.Result) {defaultR=true, group=Masks, parameterName=Primary Mask Set 2 - Selector, sortPriority=2}
   Input Low=@n78.R
   Input High=@n78.R
   Target Low=@n78.R
   Target High=@n78.R
  ) {materialFunction=&RemapValueRange}
  n80 = Multiply(A=@n53.A, B=Multiply(A=@n63, B=LinearInterpolate(A=@n71.Result, B=@n79.Result, Alpha=@n72)))
  n81 = ParticleColor()
  n82 = VectorParameter() {group=OpacityTexture_SecondaryTexture, parameterName=SecondaryTexture_Tiling N Offset, defaultValue=<rgba 1 1 1 1>}
  n83 = VectorParameter() {group=OpacityTexture_SecondaryTexture, parameterName="SecondaryTexture_Panning R= X / G = Y / B = Time", defaultValue=<rgba 0 0 1 0>}
  n84 = ScalarParameter() {defaultValue=1, group=OpacityTexture_SecondaryTexture, parameterName="SecondaryTexture_ScaleTexture = X"}
  n85 = ScalarParameter() {defaultValue=1, group=OpacityTexture_SecondaryTexture, parameterName="SecondaryTexture_ScaleTexture = Y"}
  n86 = VectorParameter() {
   group=OpacityTexture_SecondaryTexture
   parameterName="SecondaryTexture_RotationCenter R= X G= Y // B=RotationAngle"
   defaultValue=<rgba 0.5 0.5 0 0>
  }
  n87 = Multiply(
   A=ScalarParameter() {defaultValue=1, group=OpacityTexture_SecondaryTexture, parameterName=SecondaryTexture_HardnessTexture}
   B=MaterialFunctionCall(
    "Offset = 0// Mirror = 1 ForSocket 1stTextureTexture"=ScalarParameter() {defaultValue=1, group=OpacityTexture_SecondaryTexture, parameterName="Offset = 0// Mirror = 1 ForSocket SecondaryTexture"}
    Texture=TextureObjectParameter() {group=OpacityTexture_SecondaryTexture, parameterName=SecondaryTexture_OpacityTexture}
    Offset/Mirror Y 1stTexture=@n54.Offset/Mirror Y 2ndTexture
    Offset/Mirror X 1stTexture=@n54.Offset/Mirror Y 2ndTexture
    Tiling N Offset=MaterialFunctionCall(X=@n82.R, Y=@n82.R, Z=@n82.R, A=@n82.R) {materialFunction=&MakeFloat4}.Result
    Panning=@n83.RGB
    Scale=AppendVector(A=@n84, B=@n85)
    Rotation Point N Angle=@n86.RGB
    Use distortion?=StaticBoolParameter() {group=OpacityTexture_SecondaryTexture, parameterName=SecondaryTexture_Use distortion?}
   ) {materialFunction=&MF_VCVFX_BaseTextureMap}.RGB
  )
  n88 = ScalarParameter() {defaultValue=1, group=OpacityTexture_SecondaryTexture, parameterName=SecondaryTexture_Contrast}
  n89 = VectorParameter() {group=OpacityTexture_SecondaryTexture, parameterName=SecondaryTexture_RGBSelectorTextureOpacity, defaultValue=<rgba 1 1 1 1>}
  n90 = MaterialFunctionCall(
   OpacityTexture=Multiply(
    A=Power(Base=@n87, Exp=@n88)
    B=ScalarParameter() {defaultValue=1, group=OpacityTexture_SecondaryTexture, parameterName=SecondaryTexture_Emissive}
   )
   RGB Selector=@n89.RGB
   Diffuse_Texture=TextureObjectParameter() {group=OpacityTexture_Main, parameterName=SecondaryTexture_DiffuseTexture}
  ) {materialFunction=&MF_VCFX_PaintedVersions}
  n91 = LinearInterpolate(
   A=@n90.ParticleColorVersion
   B=@n90.ParticleColorVersion
   Alpha=ScalarParameter() {group=OpacityTexture_Main, parameterName="SecondaryTexture_ParticleColor  //DiffuseVersion"}
  )
  n92 = VectorParameter() {
   desc="RG controls normalized rotation angle (0 - 1) for Primary and Secondary mask respectively."
   group=Masks
   parameterName=Masks Rotation Angle
   sortPriority=8
   defaultValue=<rgba 0 0 0 1>
  }
  n93 = MaterialFunctionCall(UVs=TextureCoordinate(), "Rotation Angle (0-1)"=@n92.G) {materialFunction=&CustomRotator}
  n94 = ConstantBiasScale(None=@n93.Rotated Values) {bias=-0.5, scale=2}
  n95 = Abs(None=ComponentMask(None=@n94) {r=true})
  n96 = OneMinus(None=ComponentMask(None=@n94) {g=true})
  n97 = Multiply(A=@n95, B=@n96)
  n98 = Multiply(A=OneMinus(None=@n95), B=@n96)
  n99 = VectorParameter() {
   desc=R controls Input Low. G controls Input High. B controls Target Low. A controls Target High.
   group=Masks
   parameterName=Secondary Mask Set 2 - Remap
   sortPriority=7
   defaultValue=<rgba 0 1 0 1>
  }
  n100 = MaterialFunctionCall(
   Input=StaticComponentMaskParameter(None=MaterialFunctionCall(X=@n97, Y=@n98) {materialFunction=&MakeFloat2}.Result) {defaultR=true, group=Masks, parameterName=Secondary Mask Set 2 - Selector, sortPriority=6}
   Input Low=@n99.R
   Input High=@n99.R
   Target Low=@n99.R
   Target High=@n99.R
  ) {materialFunction=&RemapValueRange}
  n101 = ScalarParameter() {group=Masks, parameterName=SecondaryTexture_SimpleGradient//MirrorGradient}
  n102 = ComponentMask(None=@n93.Rotated Values) {r=true}
  n103 = Sine(None=@n102) {period=2}
  n104 = OneMinus(None=@n103)
  n105 = Multiply(A=@n103, B=Sine(None=ComponentMask(None=@n93.Rotated Values) {g=true}) {period=2})
  n106 = VectorParameter() {
   desc=R controls Input Low. G controls Input High. B controls Target Low. A controls Target High.
   group=Masks
   parameterName=Secondary Mask Set 1 - Remap
   sortPriority=5
   defaultValue=<rgba 0 1 0 1>
  }
  n107 = MaterialFunctionCall(
   Input=StaticComponentMaskParameter(None=MaterialFunctionCall(X=@n102, Y=@n103, Z=@n104, A=@n105) {materialFunction=&MakeFloat4}.Result) {defaultR=true, group=Masks, parameterName=Secondary Mask Set 1 - Selector, sortPriority=4}
   Input Low=@n106.R
   Input High=@n106.R
   Target Low=@n106.R
   Target High=@n106.R
  ) {materialFunction=&RemapValueRange}
  n108 = Multiply(A=@n81.A, B=Multiply(A=@n91, B=LinearInterpolate(A=@n107.Result, B=@n100.Result, Alpha=@n101)))
  n109 = ScalarParameter() {desc="0=Multiply/1=Add", group="Fusion Mode 0=M/1=A", parameterName=Main Fusion Mode_Opacity, sortPriority=1}
  MaterialFunctionCall(
   Opacity In=LinearInterpolate(A=Multiply(A=@n80, B=@n108), B=Add(A=@n80, B=@n108), Alpha=Round(None=@n109))
   DepthFade Distance=ScalarParameter() {parameterName="DepthFade Distance (S)"}
  ) {materialFunction=&MF_VCVFX_DepthFade}
  n110 = Multiply(A=@n53.RGB, B=@n63)
  n111 = Multiply(A=@n81.RGB, B=@n91)
  n112 = ScalarParameter() {desc="0=Multiply/1=Add", group="Fusion Mode 0=M/1=A", parameterName=Main Fusion Mode, sortPriority=1}
  MaterialFunctionCall(Emissive=LinearInterpolate(A=Multiply(A=@n110, B=@n111), B=Add(A=@n110, B=@n111), Alpha=Round(None=@n112))) {materialFunction=&MF_FX_EmissiveAdjustments}
 ]
}
