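# Translucent unlit Niagara master material that fakes layered depth with RGB-split copies of one texture.
# BumpOffset samples at separate depth offsets become red, green and blue layers alongside a base layer tinted by particle color.
# Named reroutes carry colors, opacities and an edge-fade UV mask with optional flipbook tiling into emissive and opacity math.
# Dynamic parameters expose texture switch, opacity and brightness controls, and MF_FX_EmissiveAdjustments processes the emissive result.
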Material /DealershipSharedLibrary/FX/Materials/MasterMaterials/General/M_BumpDepth
graph {
 properties {
  bAutomaticallySetUsageInEditor=false
  bUsedWithNiagaraMeshParticles=true
  bUsedWithNiagaraRibbons=true
  bUsedWithNiagaraSprites=true
  blendMode=BLEND_Translucent
  shadingModel=MSM_Unlit
  usedShadingModels=Unlit
 }
 nodes [
  n1 = NamedRerouteUsage()
  n2 = NamedRerouteUsage()
  n3 = Constant() {r=0.5}
  n4 = Add(A=NamedRerouteUsage(), B=NamedRerouteUsage())
  n5 = NamedRerouteUsage()
  n6 = Saturate(None=Multiply(A=NamedRerouteUsage(), B=Add(A=@n4, B=@n5)))
  n7 = LinearInterpolate(A=Multiply(A=Desaturation(None=@n2, Fraction=@n3), B=@n6), B=@n6, Alpha=NamedRerouteUsage())
  MaterialFunctionCall(Emissive=Multiply(A=Add(A=@n1, B=@n7), B=NamedRerouteUsage())) {materialFunction=&MF_FX_EmissiveAdjustments}
  n8 = DynamicParameter() {paramNames [TextureSwitch, Opacity, Brightness, RGB_Brightness]}
  NamedRerouteDeclaration(None=@n8.TextureSwitch) {name=Dyn TextureSwitch, nodeColor=<rgba 0.8047 1 0 1>}
  NamedRerouteDeclaration(None=@n8.Opacity) {name=Dyn Opacity, nodeColor=<rgba 0.875 0 1 1>}
  n9 = TextureSampleParameter2D(UVs=BumpOffset(Coordinate=TextureCoordinate(), Height=ScalarParameter() {defaultValue=-1.5, group=DepthOffset, parameterName=GOffset})) {parameterName=Texture, texture=/DealershipSharedLibrary/Assets/Sets/RetroTech/T_RetroTech_Triangles_RGB}
  n10 = NamedRerouteUsage()
  n11 = NamedRerouteDeclaration(None=LinearInterpolate(A=@n9.R, B=@n9.R, Alpha=@n10)) {name=GOpacity, nodeColor=<rgba 1 1 1 1>}
  n12 = Multiply(A=LinearInterpolate(A=@n9.A, B=@n9.A, Alpha=@n10), B=NamedRerouteUsage())
  NamedRerouteDeclaration(None=Multiply(A=Add(A=@n11, B=@n12), B=Constant3Vector() {constant=<rgba 0 1 0 0>})) {name=GColor, nodeColor=<rgba 1 1 1 1>}
  n13 = TextureSampleParameter2D(UVs=BumpOffset(Coordinate=TextureCoordinate(), Height=ScalarParameter() {defaultValue=-2, group=DepthOffset, parameterName=BOffset})) {parameterName=Texture, texture=&T_RetroTech_Triangles_RGB}
  n14 = NamedRerouteUsage()
  n15 = NamedRerouteDeclaration(None=LinearInterpolate(A=@n13.R, B=@n13.R, Alpha=@n14)) {name=BOpacity, nodeColor=<rgba 1 1 1 1>}
  n16 = Multiply(A=LinearInterpolate(A=@n13.A, B=@n13.A, Alpha=@n14), B=NamedRerouteUsage())
  NamedRerouteDeclaration(None=Multiply(A=Add(A=@n15, B=@n16), B=Constant3Vector() {constant=<rgba 0 0 1 0>})) {name=BColor, nodeColor=<rgba 1 1 1 1>}
  n17 = TextureCoordinate()
  n18 = ScalarParameter() {defaultValue=4, group=Flipbook, parameterName=Columns, sortPriority=0}
  n19 = ScalarParameter() {defaultValue=4, group=Flipbook, parameterName=Rows}
  n20 = Frac(None=Multiply(A=TextureCoordinate(), B=AppendVector(A=@n18, B=@n19)))
  n21 = LinearInterpolate(A=@n17, B=@n20, Alpha=ScalarParameter() {group=Flipbook, parameterName=FlipbookMask, sliderMax=1})
  n22 = ComponentMask(None=@n21) {r=true}
  n23 = Saturate(None=ConstantBiasScale(None=OneMinus(None=@n22)) {bias=-0.025, scale=20})
  n24 = Saturate(None=ConstantBiasScale(None=@n22) {bias=-0.025, scale=20})
  n25 = ComponentMask(None=@n21) {g=true}
  n26 = Saturate(None=ConstantBiasScale(None=OneMinus(None=@n25)) {bias=-0.025, scale=20})
  n27 = Saturate(None=ConstantBiasScale(None=@n25) {bias=-0.025, scale=20})
  NamedRerouteDeclaration(None=Multiply(A=Multiply(A=@n23, B=@n24), B=Multiply(A=@n26, B=@n27))) {name=UVMask, nodeColor=<rgba 1 1 1 1>}
  n28 = TextureSampleParameter2D(UVs=BumpOffset(Coordinate=TextureCoordinate(), Height=ScalarParameter() {defaultValue=-1, group=DepthOffset, parameterName=ROffset})) {parameterName=Texture, texture=&T_RetroTech_Triangles_RGB}
  n29 = NamedRerouteUsage()
  n30 = NamedRerouteDeclaration(None=LinearInterpolate(A=@n28.R, B=@n28.R, Alpha=@n29)) {name=ROpacity, nodeColor=<rgba 1 1 1 1>}
  n31 = NamedRerouteUsage()
  n32 = ScalarParameter() {defaultValue=0.05, parameterName=RGBOpacityIntensity}
  n33 = Multiply(
   A=LinearInterpolate(A=@n28.A, B=@n28.A, Alpha=@n29)
   B=NamedRerouteDeclaration(None=Multiply(A=@n31, B=@n32) {constB=0.05}) {name=RGBOpacity, nodeColor=<rgba 1 1 1 1>}
  )
  NamedRerouteDeclaration(None=Multiply(A=Add(A=@n30, B=@n33), B=Constant3Vector() {constant=<rgba 1 0 0 0>})) {name=RColor, nodeColor=<rgba 1 1 1 1>}
  n34 = TextureSampleParameter2D(UVs=TextureCoordinate()) {parameterName=Texture, texture=&T_RetroTech_Triangles_RGB}
  n35 = NamedRerouteUsage()
  n36 = LinearInterpolate(A=@n34.R, B=@n34.R, Alpha=@n35)
  n37 = ParticleColor()
  n38 = NamedRerouteDeclaration(None=@n37.RGB) {name=ParticleColor, nodeColor=<rgba 1 1 1 1>}
  n39 = Multiply(A=LinearInterpolate(A=@n34.A, B=@n34.A, Alpha=@n35), B=NamedRerouteUsage())
  NamedRerouteDeclaration(None=Add(A=Multiply(A=@n38, B=@n36), B=Multiply(A=@n39, B=@n38))) {name=Color, nodeColor=<rgba 1 1 1 1>}
  NamedRerouteDeclaration(None=@n8.Brightness) {name=DYN_Brightness, nodeColor=<rgba 0 1 0.5547 1>}
  NamedRerouteDeclaration(None=@n8.RGB_Brightness) {name=DYN_RGB_Brightness, nodeColor=<rgba 1 0.2344 0 1>}
  NamedRerouteDeclaration(None=@n37.A) {name=ParticleAlpha, nodeColor=<rgba 1 1 1 1>}
  NamedRerouteDeclaration(None=Add(A=@n36, B=@n39)) {name=TexOpacity, nodeColor=<rgba 1 1 1 1>}
  n40 = NamedRerouteUsage()
  n41 = NamedRerouteUsage()
  n42 = Add(A=NamedRerouteUsage(), B=Add(A=Add(A=@n40, B=@n41), B=NamedRerouteUsage()))
  n43 = NamedRerouteUsage()
  Saturate(None=Multiply(A=Multiply(A=@n42, B=@n43), B=NamedRerouteUsage()))
  StaticSwitchParameter(True=@n20, False=@n17) {group=Mask, parameterName=FlipBook}
 ]
}
