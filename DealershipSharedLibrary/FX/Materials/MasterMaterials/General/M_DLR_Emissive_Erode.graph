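# Unlit translucent master material for Niagara sprite and mesh particles, producing emissive effects with distortion and alpha erosion.
# A panned normal map's red channel offsets the diffuse texture UVs.
# Diffuse alpha passes through a SmoothStep erode with soft edge, then opacity and depth fade.
# Particle color drives flood and texture-tinted emissive through MF_FX_EmissiveAdjustments, with two Dynamic Parameter sets exposing per-particle controls.
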Material /DealershipSharedLibrary/FX/Materials/MasterMaterials/General/M_DLR_Emissive_Erode
graph {
 properties {
  bCastRayTracedShadows=false
  bUseTranslucencyVertexFog=false
  bUsedWithNiagaraMeshParticles=true
  bUsedWithNiagaraSprites=true
  blendMode=BLEND_Translucent
  pixelDepthOffsetMode=PDOM_Legacy
  shadingModel=MSM_Unlit
  translucencyDirectionalLightingIntensity=0.1
  translucentSelfShadowDensityScale=0
  translucentSelfShadowSecondDensityScale=1
  translucentShadowDensityScale=0
  translucentShadowStartOffset=1
  usedShadingModels=Unlit
  lightmassSettings {
   bCastShadowAsMasked=false
   bOverrideCastShadowAsMasked=false
   bOverrideDiffuseBoost=false
   bOverrideEmissiveBoost=false
   bOverrideExportResolutionScale=false
   diffuseBoost=8
   emissiveBoost=1
   exportResolutionScale=1
  }
  translucentMultipleScatteringExtinction=<rgba 1 0.8771 0.6778 1>
 }
 nodes [
  n1 = DynamicParameter() {paramNames [Opacity, Erode, Refract_Pan, Refract_Intensity]}
  n2 = Multiply(A=Constant() {desc=Refract_Intensity, r=0.5}, B=@n1.Refract_Intensity)
  n3 = TextureSampleParameter2D(UVs=Panner(Coordinate=TextureCoordinate() {uTiling=0.4, vTiling=0.4}, Speed=@n1.Refract_Pan) {speedX=1.5, speedY=1.5}) {parameterName=Distort, samplerType=SAMPLERTYPE_Normal, texture=/DealershipSharedLibrary/FX/Textures/Normal/T_VCVFX_Liquid_Tile_N}
  n4 = TextureSampleParameter2D(UVs=Add(A=TextureCoordinate(), B=Multiply(A=@n2, B=@n3.R))) {parameterName=Diffuse, texture=/DealershipSharedLibrary/FX/Textures/Diffuse/T_VCVFX_DefaultDiffuse_D}
  n5 = OneMinus(None=Multiply(A=Constant() {desc=Erode, r=1}, B=@n1.Erode))
  n6 = DynamicParameter() {parameterIndex=1, paramNames [Flood, Emissive, Soft_Edge, ""]}
  n7 = Add(A=@n5, B=Multiply(A=Constant() {desc=SoftEdge}, B=@n6.Soft_Edge))
  DepthFade(
   Opacity=Multiply(
    A=MaterialFunctionCall(Alpha=@n4.A, Min=@n5, Max=@n7) {materialFunction=&SmoothStep}.Result
    B=@n1.Opacity
   )
   FadeDistance=ScalarParameter() {defaultValue=300, parameterName=DepthFade}
  ) {fadeDistanceDefault=300}
  n8 = ParticleColor()
  n9 = Multiply(A=@n8.RGB, B=@n4.RGB)
  MaterialFunctionCall(Emissive=Add(A=Multiply(A=@n8.RGB, B=@n6.Flood), B=Multiply(A=@n9, B=@n6.Emissive))) {materialFunction=&MF_FX_EmissiveAdjustments}
 ]
}
