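# Translucent unlit master material for gradient light beams on Niagara sprites and mesh particles.
# The beam mask multiplies a UV gradient, a soft or stepped cross-beam SphereMask, and optional local-axis and bottom fades, with depth fade on opacity.
# Color switches between time-of-day-scaled particle color and a two-color gradient.
# SplineThicken can reshape the beam through world position offset, with dynamic parameters exposing widths and gradient controls.
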Material /DealershipSharedLibrary/FX/Materials/MasterMaterials/General/M_DLR_GradientBeam
graph {
 properties {
  bUsedWithNanite=true
  bUsedWithNiagaraMeshParticles=true
  bUsedWithNiagaraSprites=true
  blendMode=BLEND_Translucent
  shadingModel=MSM_Unlit
  usedShadingModels=Unlit
 }
 nodes [
  n1 = Multiply(A=ParticleColor().A, B=NamedRerouteUsage())
  MaterialFunctionCall(
   Opacity In=@n1
   DepthFade Distance=ScalarParameter() {defaultValue=100, group="!Master Parameters", parameterName=Depth Fade Distance, sortPriority=1}
  ) {materialFunction=&MF_VCVFX_DepthFade}
  MaterialFunctionCall(Emissive=NamedRerouteUsage()) {materialFunction=&MF_FX_EmissiveAdjustments}
  n2 = Multiply(A=NamedRerouteUsage(), B=ScalarParameter() {defaultValue=100, group=WPO, parameterName=WPO WidthTip})
  n3 = Multiply(A=@n2, B=VertexColor().R)
  n4 = MaterialFunctionCall(
   WidthBase=Multiply(A=NamedRerouteUsage(), B=ScalarParameter() {defaultValue=50, group=WPO, parameterName=WPO WidthBase})
   WidthTip=If(A=ScalarParameter() {defaultValue=1, group=WPO, parameterName=WPO Use Vertex Color?}, "A > B"=@n3, "A == B"=@n3, "A < B"=@n2) {constB=1}
  ) {materialFunction=&SplineThicken}
  NamedRerouteDeclaration(None=StaticSwitchParameter(True=@n4.UVs with Parallax, False=TextureCoordinate()) {defaultValue=true, group=WPO, parameterName=WPO Squish Beam with WPO?, sortPriority=0}) {name=SplineThicken_UVs, nodeColor=<rgba 1 0.7969 0 1>}
  n5 = DynamicParameter() {defaultValue=<rgba 0 1 0.5 1>, paramNames [UV Gradient Offset, UV Gradient Amount, Local Aligned Gradient Offset, Param4]}
  NamedRerouteDeclaration(None=@n5.UV Gradient Offset) {name=DP_UV_Gradient_Offset, nodeColor=<rgba 0.2422 1 0 1>}
  n6 = ParticleColor()
  n7 = MaterialFunctionCall(
   Value_Morning=ScalarParameter() {defaultValue=1, group=Color, parameterName=TOD Morning Emissive Amount}
   Value_Day=ScalarParameter() {defaultValue=1, group=Color, parameterName=TOD Day Emissive Amount}
   Value_Evening=ScalarParameter() {defaultValue=1, group=Color, parameterName=TOD Evening Emissive Amount}
   Value_Night=ScalarParameter() {defaultValue=1, group=Color, parameterName=TOD Night Emissive Amount}
  ) {materialFunction=&Scalar_Time_of_Day}
  n8 = VectorParameter() {group=Color, parameterName=UpperColor, sortPriority=20, defaultValue=<rgba 0 0.7136 1 1>}
  n9 = VectorParameter() {group=Color, parameterName=LowerColor, sortPriority=20, defaultValue=<rgba 1 0.0021 0.6002 1>}
  n10 = Add(A=TextureCoordinate(), B=ScalarParameter() {group=Color, parameterName=Color Gradient Offset, sortPriority=30})
  n11 = ScalarParameter() {defaultValue=1, group=Color, parameterName=Color Gradient Contrast, sortPriority=30}
  n12 = ComponentMask(None=Saturate(None=Power(Base=@n10, Exp=@n11))) {g=true}
  NamedRerouteDeclaration(None=StaticSwitchParameter(True=LinearInterpolate(A=@n8.RGB, B=@n9.RGB, Alpha=@n12), False=Multiply(A=@n6.RGB, B=@n7.Result)) {group=Color, parameterName=Use Particle Color or Color Gradient?, sortPriority=1}) {name=Color, nodeColor=<rgba 0.3125 1 0 1>}
  NamedRerouteDeclaration(None=@n5.UV Gradient Amount) {name=DP_UV_Gradient_Amount, nodeColor=<rgba 0.2422 1 0 1>}
  NamedRerouteDeclaration(None=@n5.Local Aligned Gradient Offset) {name=DP_Local_Aligned_Gradient_Offset, nodeColor=<rgba 0.2422 1 0 1>}
  n13 = DynamicParameter() {parameterIndex=1, paramNames [WPO WidthBaseMult, WPO WidthTipMult, FlickerFade, FlickerSpeed]}
  NamedRerouteDeclaration(None=@n13.WPO WidthBaseMult) {name=DP_WPO_WidthBaseMult, nodeColor=<rgba 1 0.7969 0 1>}
  NamedRerouteDeclaration(None=@n13.WPO WidthTipMult) {name=DP_WPO_TipBaseMult, nodeColor=<rgba 1 0.7969 0 1>}
  NamedRerouteDeclaration(None=StaticSwitchParameter(True=@n4.WorldPositionOffset, False=VertexNormalWS()) {defaultValue=true, group=WPO, parameterName=WPO Squish Beam with WPO?, sortPriority=0}) {name=WPO, nodeColor=<rgba 0 1 0.1562 1>}
  n14 = LocalPosition()
  n15 = VectorParameter() {group=Local Aligned Gradient, parameterName=Local Aligned Gradient Axis, sortPriority=1, defaultValue=<rgba 0 0 -1 0>}
  n16 = Multiply(A=ObjectBounds(), B=@n15.RGB)
  n17 = Add(A=ComponentMask(None=Divide(A=@n14.XYZ, B=@n16)) {b=true}, B=NamedRerouteUsage())
  n18 = ScalarParameter() {defaultValue=2, group=Local Aligned Gradient, parameterName=Local ALigned Gradient Height, sortPriority=2}
  NamedRerouteDeclaration(None=StaticSwitchParameter(True=Divide(A=@n17, B=@n18), False=Constant() {r=1}) {group=Local Aligned Gradient, parameterName=Local Aligned Gradient Mask?, sortPriority=0}) {name=Local Aligned Gradient, nodeColor=<rgba 1 0 0.8437 1>}
  n19 = ScalarParameter() {defaultValue=3, group=Bottom Fade, parameterName=BottomFadeSize}
  n20 = ComponentMask(None=NamedRerouteUsage()) {g=true}
  n21 = Power(
   Base=Saturate(None=LinearInterpolate(A=@n19, Alpha=@n20) {constA=1, constB=0})
   Exp=ScalarParameter() {defaultValue=1, group=Bottom Fade, parameterName=Bottom Fade Power}
  )
  n22 = ScalarParameter() {group=Bottom Fade, parameterName=Bottom Fade?, sortPriority=0}
  NamedRerouteDeclaration(None=Clamp(None=LinearInterpolate(B=@n21, Alpha=@n22) {constA=1})) {name=Bottom Fade, nodeColor=<rgba 0 1 0.2734 1>}
  n23 = Add(A=ComponentMask(None=NamedRerouteUsage()) {g=true}, B=NamedRerouteUsage())
  n24 = NamedRerouteUsage()
  n25 = ScalarParameter() {group=Stepped Gradient, parameterName=Stepped Gradient?, sortPriority=0}
  n26 = Constant() {r=0.5}
  n27 = SphereMask(A=ComponentMask(None=NamedRerouteUsage()) {r=true}, B=@n26, Radius=@n26) {hardnessPercent=0}
  n28 = Multiply(
   A=@n27
   B=MaterialFunctionCall!(
    Input=Clamp(None=@n27)
    Input Low=Constant()
    Input High=Constant() {r=1}
    Target Low=Constant()
    Target High=ScalarParameter() {defaultValue=2, group=Stepped Gradient, parameterName=Stepped Gradient High}
    Steps=ScalarParameter() {defaultValue=4, group=Stepped Gradient, parameterName=Stepped Gradient Steps}
   ) {materialFunction=/DealershipSharedLibrary/FX/Materials/MaterialFunctions/MF_DLR_RemapValueRangeStepped}.Stepped
  )
  n29 = Multiply(A=Clamp(None=Power(Base=@n23, Exp=@n24)), B=Clamp(None=If(A=@n25, "A > B"=@n28, "A == B"=@n28, "A < B"=@n27) {constB=1}))
  n30 = NamedRerouteUsage()
  NamedRerouteDeclaration(None=Multiply(A=Multiply(A=@n29, B=@n30), B=NamedRerouteUsage())) {name=Beam_Gradient_Out, nodeColor=<rgba 1 0 0.0469 1>}
  NamedRerouteUsage()
  Reroute(None=@n1)
 ]
}
