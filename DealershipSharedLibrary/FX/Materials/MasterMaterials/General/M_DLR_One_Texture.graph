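# Translucent unlit master material for Niagara VFX, built around one rotatable, panning texture with optional UV distortion.
# Opacity multiplies that channel-selected, invertible texture by a texture or sphere mask eroded via dynamic-parameter threshold, then applies particle alpha, optional edge mask and depth fade.
# Emissive combines particle color, dynamic emission and optional texture modulation, with an IsPainted dynamic parameter for tint selection.
# Distortion tiling and offset reuse their U/X value on both axes.
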Material /DealershipSharedLibrary/FX/Materials/MasterMaterials/General/M_DLR_One_Texture
graph {
 properties {
  bAutomaticallySetUsageInEditor=false
  bEnableResponsiveAA=true
  bUsedWithNiagaraMeshParticles=true
  bUsedWithNiagaraRibbons=true
  bUsedWithNiagaraSprites=true
  blendMode=BLEND_Translucent
  shadingModel=MSM_Unlit
  usedShadingModels=Unlit
 }
 nodes [
  n1 = Saturate(None=ScalarParameter() {group=BaseTexture, parameterName="InvertTexture (If 0  No - If 1 Yes)", sortPriority=16})
  n2 = TextureCoordinate()
  n3 = MaterialFunctionCall(
   UVs=@n2
   Rotation Center=VectorParameter() {group=Distortion, parameterName=Texture_Rotation_Center, sortPriority=20}.RGB
   "Rotation Angle (0-1)"=ScalarParameter() {defaultValue=0.25, group=BaseTexture, parameterName=Texture_Angle_Offset, sortPriority=24}
  ) {materialFunction=&CustomRotator}
  n4 = Multiply(A=Time(), B=ScalarParameter() {group=BaseTexture, parameterName=Panner_Time})
  n5 = VectorParameter() {group=BaseTexture, parameterName=Panner_Direction, sortPriority=28, defaultValue=<rgba -1 0 0 0>}
  n6 = TextureSampleParameter2D(UVs=Add(A=Panner(Coordinate=@n3.Rotated Values, Time=@n4, Speed=@n5.RGB) {speedX=-1}, B=NamedRerouteUsage())) {group=BaseTexture, parameterName=Texture, sortPriority=0, texture=/Game/Effects/Fort_Effects/Textures/Tile/T_Glass_Cracks_M}
  n7 = Multiply(
   A=ScalarParameter() {defaultValue=50, group=BaseTexture, parameterName=Alpha_Multiply, sortPriority=12}
   B=Power(
    Base=ChannelMaskParameter(None=@n6.RGBA) {group=BaseTexture, parameterName=Texture Channel, sortPriority=4}.R
    Exp=ScalarParameter() {defaultValue=2, group=BaseTexture, parameterName=Power_Exp, sortPriority=8}
   )
  )
  n8 = OneMinus(None=@n7)
  n9 = Saturate(None=ScalarParameter() {group=BaseTexture, parameterName="InvertTexture (If 0  No - If 1 Yes)", sortPriority=16})
  n10 = TextureSampleParameter2D(UVs=Add(A=@n2, B=NamedRerouteUsage())) {
   group=Mask
   parameterName=AlphaMask
   samplerType=SAMPLERTYPE_Grayscale
   sortPriority=4
   texture=/Game/Effects/Fort_Effects/Textures/Tile/T_Concrete_Crack_A
  }
  n11 = StaticSwitchParameter(
   True=DotProduct(A=@n10.RGBA, B=VectorParameter() {group=Mask, parameterName=Alpha Mask - Texture Channel Select, defaultValue=<rgba 1 0 0 0>}.RGBA)
   False=SphereMask(
    A=TextureCoordinate()
    B=Constant2Vector() {g=0.5, r=0.5}
    Radius=ScalarParameter() {defaultValue=0.5, group=Mask, parameterName=SphereMask_Radius, sortPriority=8}
    Hardness=ScalarParameter() {group=Mask, parameterName=SphereMask_Hardness, sortPriority=12}
   )
  ) {defaultValue=true, group=Mask, parameterName=Use_2ndaryTextureAlpha, sortPriority=0}
  n12 = ScalarParameter() {defaultValue=1, group=Mask, parameterName=MaskPower, sortPriority=16}
  n13 = DynamicParameter() {defaultValue=<rgba 1 0 1 0>, paramNames [AlphaThreshold, Param2, Param3, Param4]}
  n14 = ScalarParameter() {group=Mask, parameterName=PreDynParamAlphaThresholdAdd, sortPriority=24}
  n15 = DynamicParameter() {parameterIndex=1, defaultValue=<rgba 1 1 0 0>, paramNames [Emission, SecondaryAlphaThreshold, IsPainted, Param4]}
  n16 = ScalarParameter() {defaultValue=-1, group=Mask, parameterName=AlphaThreshold, sortPriority=28}
  n17 = Subtract(
   A=Saturate(None=Power(Base=@n11, Exp=@n12) {constExponent=1})
   B=StaticSwitchParameter(True=Add(A=@n13.AlphaThreshold, B=@n14), False=Add(A=@n15.SecondaryAlphaThreshold, B=@n16)) {group=Mask, parameterName=Dynamic Alpha Parameter, sortPriority=20}
  ) {constB=0}
  n18 = OneMinus(None=@n17)
  n19 = Multiply(A=Saturate(None=If(A=@n1, "A > B"=@n8, "A == B"=@n7, "A < B"=@n7)), B=Saturate(None=If(A=@n9, "A > B"=@n18, "A == B"=@n17, "A < B"=@n17)))
  n20 = NamedRerouteUsage()
  LinearInterpolate(B=@n19, Alpha=@n20)
  n21 = ParticleColor()
  n22 = ScalarParameter() {parameterName=UseEdgeMask}
  n23 = Multiply(A=@n19, B=@n20)
  MaterialFunctionCall(
   Opacity In=Multiply(A=@n21.A, B=If(A=@n22, "A > B"=@n23, "A == B"=@n19, "A < B"=@n19))
   DepthFade Distance=ScalarParameter() {defaultValue=100, group=BaseTexture, parameterName=FadeDistance, sortPriority=36}
  ) {materialFunction=&MF_VCVFX_DepthFade}
  n24 = Multiply(A=@n15.Emission, B=NamedRerouteUsage())
  n25 = ScalarParameter() {group=Emissive, parameterName=Emission - Use Texture Lerp}
  n26 = Multiply(
   A=@n21.RGB
   B=Multiply(
    A=LinearInterpolate(A=@n15.Emission, B=@n24, Alpha=@n25)
    B=ScalarParameter() {defaultValue=1, group=Color, parameterName=EmissionMult, sortPriority=4}
   )
  )
  n27 = Multiply(A=VectorParameter() {parameterName=Emissive Color, defaultValue=<rgba 1 1 1 1>}.RGB, B=@n26)
  MaterialFunctionCall(Emissive=If(A=NamedRerouteUsage(), "A > B"=@n26, "A == B"=@n27, "A < B"=@n27)) {materialFunction=&MF_FX_EmissiveAdjustments}
  n28 = Power(
   Base=DotProduct(A=@n6.RGBA, B=VectorParameter() {group=Emissive, parameterName=Emissive -Texture Channel Picker, defaultValue=<rgba 1 0 0 0>}.RGBA)
   Exp=ScalarParameter() {defaultValue=1, group=Emissive, parameterName=Emissive - Texture Power}
  )
  n29 = ScalarParameter() {defaultValue=1, group=Emissive, parameterName=Emissive - Texture Mult}
  NamedRerouteDeclaration(None=Subtract(A=Multiply(A=@n28, B=@n29), B=ScalarParameter() {defaultValue=1, group=Emissive, parameterName=Emissive - Texture Erode})) {name=Emissive - Texture, nodeColor=<rgba 0 1 0.8916 1>}
  n30 = VectorParameter() {group=Distortion, parameterName=Distortion - Tile, sortPriority=20, channelNames {a="", b="", g=Tile V, r=Tile U}}
  n31 = Multiply(A=@n2, B=AppendVector(A=@n30.Tile U, B=@n30.Tile U))
  n32 = DynamicParameter() {parameterIndex=2, defaultValue=<rgba 0 0 1 0>, paramNames [Distortion, Distortion - Offset X, Distortion - Offset Y, Param4]}
  n33 = AppendVector(A=@n32.Distortion - Offset X, B=@n32.Distortion - Offset X)
  n34 = ComponentMask(None=TextureSampleParameter2D(UVs=Add(A=@n31, B=@n33)) {group=Distortion, parameterName=Distortion - Texture}.RGB) {g=true, r=true}
  n35 = Constant()
  NamedRerouteDeclaration(None=If(
    A=ScalarParameter() {group=Distortion, parameterName=Use Distortion}
    "A > B"=Multiply(A=@n34, B=@n32.Distortion)
    "A == B"=@n35
    "A < B"=@n35
   )) {name=UV Distortion, nodeColor=<rgba 0.4531 0 1 1>}
  n36 = TextureCoordinate() {uTiling=0.5, vTiling=0.5}
  n37 = Multiply(A=Sine(None=ComponentMask(None=@n36) {r=true}), B=Sine(None=ComponentMask(None=@n36) {g=true}))
  NamedRerouteDeclaration(None=Saturate(None=Multiply(A=@n37) {constB=2})) {name=EdgeMask, nodeColor=<rgba 0.1584 0.0061 1 1>}
  NamedRerouteDeclaration(None=@n15.IsPainted) {name=DP_IsPainted, nodeColor=<rgba 0.9544 0 1 1>}
 ]
}
