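# M_DLR_SimplePBR is a Default Lit master material for Niagara mesh, sprite and ribbon VFX in the Dealership shared library.
# Named reroutes carry base color tinted by particle or parameter color, emissive through MF_FX_EmissiveAdjustments, and two summed opacity masks scaled by particle alpha.
# Static switches choose texture or constant for specular, roughness, normal, emissive and opacity.
# A Dynamic Parameter supplies per-particle bIsPainted and Metallic values.
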Material /DealershipSharedLibrary/FX/Materials/MasterMaterials/General/M_DLR_SimplePBR
graph {
 properties {
  bAutomaticallySetUsageInEditor=false
  bUsedWithNiagaraMeshParticles=true
  bUsedWithNiagaraRibbons=true
  bUsedWithNiagaraSprites=true
  previewMesh=/Engine/Transient.MaterialEditorMeshComponent_7
  usedShadingModels=Default Lit
 }
 nodes [
  n1 = TextureSampleParameter2D() {group=Base Color, parameterName=Base Color - Texture, texture=/DealershipSharedLibrary/Assets/Sets/Daikon/T_VCVFX_Boost_Daikon_SnoBro_D}
  n2 = LinearInterpolate(
   A=@n1.RGB
   B=ChannelMaskParameter(None=@n1.RGB) {group=Base Color, parameterName=Base Color - Texture Channel}.R
   Alpha=ScalarParameter() {group=Base Color, parameterName=Base Color - Use Texture Channel?}
  )
  n3 = VectorParameter() {desc=0, parameterName=Base Color, defaultValue=<rgba 0.1667 0.1667 0.1667 1>}
  n4 = If(
   A=ScalarParameter() {defaultValue=1, parameterName=Base Color - Use Particle Color?}
   "A > B"=ParticleColor() {desc=1}.RGB
   "A == B"=@n3.RGB
   "A < B"=@n3.RGB
  )
  NamedRerouteDeclaration(None=LinearInterpolate(A=@n2, B=Multiply(A=@n4, B=@n2), Alpha=NamedRerouteUsage())) {name=Base Color, nodeColor=<rgba 1 0.8672 0 1>}
  n5 = TextureSampleParameter2D() {
   group=Specular
   parameterName=Specular - Texture
   samplerType=SAMPLERTYPE_Masks
   texture=/DealershipSharedLibrary/Assets/Sets/Fruits/Textures/T_TroutWrist_Contrail_S
  }
  NamedRerouteDeclaration(None=StaticSwitchParameter(True=@n5.R, False=ScalarParameter() {group=Specular, parameterName=Specular - Constant}) {group=Specular, parameterName=Specular - Use Texture?}) {name=Specular, nodeColor=<rgba 0.5469 0 1 1>}
  n6 = TextureSampleParameter2D() {group=Emissive, parameterName=Emissive - Texture, texture=/DealershipSharedLibrary/FX/Textures/Noises/T_Energy}
  n7 = StaticSwitchParameter(
   True=LinearInterpolate(
    A=@n6.RGB
    B=ChannelMaskParameter(None=@n6.RGB) {group=Emissive, parameterName=Emissive - Texture Channel}.R
    Alpha=ScalarParameter() {group=Emissive, parameterName=Emissive- Use Texture Channel?}
   )
   False=ScalarParameter() {group=Emissive, parameterName=Emissive - Constant Value}
  ) {group=Emissive, parameterName=Emissive - Use Texture?}
  n8 = ParticleColor()
  n9 = LinearInterpolate(
   A=Multiply(A=@n7, B=@n8.RGB)
   B=Add(A=@n7, B=@n8.RGB)
   Alpha=ScalarParameter() {group=Emissive, parameterName="Emissive - Add Particle Color? (0=multiply 1=add))"}
  )
  n10 = ScalarParameter() {group=Emissive, parameterName=Emissive - Use Particle Color?}
  NamedRerouteDeclaration(None=MaterialFunctionCall(Emissive=LinearInterpolate(A=@n7, B=@n9, Alpha=@n10)) {materialFunction=&MF_FX_EmissiveAdjustments}.Result) {name=Emissive, nodeColor=<rgba 0 1 0.2266 1>}
  n11 = TextureSampleParameter2D() {
   group=Normal
   parameterName=Normal - Texture
   samplerType=SAMPLERTYPE_Normal
   texture=/DealershipSharedLibrary/Assets/Sets/Daikon/T_VCVFX_Boost_Daikon_SnoBro_N
  }
  NamedRerouteDeclaration(None=StaticSwitchParameter(True=@n11.RGB, False=VectorParameter() {group=Normal, parameterName="Normal -  Constant Value", defaultValue=<rgba 0 0 1 1>}.RGB) {group=Normal, parameterName=Normal - Use Texture?}) {name=Normals, nodeColor=<rgba 1 0 0.0938 1>}
  n12 = TextureSampleParameter2D() {group=Opacity, parameterName=Opacity - Texture, texture=/DealershipSharedLibrary/FX/Textures/Gradient_Masks/T_VCVFX_ImpactFilled_01_A}
  n13 = StaticSwitchParameter(
   True=LinearInterpolate(
    A=@n12.RGB
    B=ChannelMaskParameter(None=@n12.RGB) {group=Opacity, parameterName=Opacity - Texture Channel}.R
    Alpha=ScalarParameter() {defaultValue=1, group=Opacity, parameterName=Opacity - Use Texture Channel?}
   )
   False=ScalarParameter() {group=Opacity, parameterName=Opacity - Constant Value}
  ) {group=Opacity, parameterName=Opacity - Use Texture?}
  n14 = TextureSampleParameter2D() {group=Opacity, parameterName=Opacity - Texture 2, texture=&T_VCVFX_ImpactFilled_01_A}
  n15 = StaticSwitchParameter(
   True=LinearInterpolate(
    A=@n14.RGB
    B=ChannelMaskParameter(None=@n14.RGB) {group=Opacity, parameterName=Opacity - Texture 2 Channel}.R
    Alpha=ScalarParameter() {defaultValue=1, group=Opacity, parameterName=Opacity - Use Texture 2 Channel?}
   )
   False=Constant()
  ) {group=Opacity, parameterName=Opacity - Use Texture 2?}
  NamedRerouteDeclaration(None=Multiply(A=ParticleColor().A, B=Add(A=@n13, B=@n15))) {name=Opacity, nodeColor=<rgba 0 0.4141 1 1>}
  n16 = TextureSampleParameter2D() {
   group=Roughness
   parameterName=Roughness - Texture
   samplerType=SAMPLERTYPE_Masks
   texture=&T_TroutWrist_Contrail_S
  }
  NamedRerouteDeclaration(None=StaticSwitchParameter(True=@n16.G, False=ScalarParameter() {group=Roughness, parameterName=Roughness - Constant}) {group=Roughness, parameterName=Roughness - Use Texture?}) {name=Roughness, nodeColor=<rgba 0.1391 0.158 1 1>}
  n17 = DynamicParameter() {defaultValue=<rgba 0 0 1 1>, paramNames [bIsPainted,Metallic,Param3,Param4]}
  NamedRerouteDeclaration(None=@n17.bIsPainted) {name=BlsPainted, nodeColor=<rgba 0.2891 1 0 1>}
  NamedRerouteDeclaration(None=@n17.Metallic) {name=Metallic, nodeColor=<rgba 1 0 0.6094 1>}
  NamedRerouteUsage()
  NamedRerouteUsage()
  NamedRerouteUsage()
  NamedRerouteUsage()
  NamedRerouteUsage()
  NamedRerouteUsage()
  NamedRerouteUsage()
 ]
}
