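# Translucent, unlit master material for flipbook effects on Niagara sprites and mesh particles, defaulting to a packed fire texture.
# A SubUV function samples the tiled texture, with dynamic parameters declared for sub-image counts, frame and packed channel.
# Optional channel masks and power shape the flipbook; particle color drives emissive adjustments and opacity, with depth fade and dithered mask variants.
# The time-of-day emissive multiplier feeds its Morning channel to all four period inputs.
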Material /DealershipSharedLibrary/FX/Materials/MasterMaterials/General/M_DLR_Simple_Flipbook
graph {
 properties {
  bUsedWithNiagaraMeshParticles=true
  bUsedWithNiagaraSprites=true
  blendMode=BLEND_Translucent
  shadingModel=MSM_Unlit
  usedShadingModels=Unlit
 }
 nodes [
  n1 = TextureObjectParameter() {group=Texture, parameterName=Flipbook Texture, sortPriority=1, texture=/Game/Effects/Fort_Effects/Textures/Fire/2DGasTorch_Flipbk_packed}
  n2 = ScalarParameter() {defaultValue=1, group=Texture, parameterName=UV Tiling X}
  n3 = ScalarParameter() {defaultValue=1, group=Texture, parameterName=UV Tiling Y}
  n4 = NamedRerouteUsage()
  n5 = NamedRerouteUsage()
  n6 = MaterialFunctionCall!(
   Texture=@n1
   UVs=Multiply(A=TextureCoordinate(), B=AppendVector(A=@n2, B=@n3))
   SubImages=AppendVector(A=@n4, B=Add(A=@n4, B=@n5))
   Frame=NamedRerouteUsage()
  ) {materialFunction=/Engine/Functions/Engine_MaterialFunctions03/Texturing/SubUV_Function}
  NamedRerouteDeclaration(None=@n6.Alpha) {name=SubUVFunction Alpha, nodeColor=<rgba 1 0.9399 0.8741 1>}
  n7 = DynamicParameter() {defaultValue=<rgba 1 1 1 0>, paramNames [Packed Channel, SubUV_Images, SubUV_Frame, SubUV_ImagesY]}
  NamedRerouteDeclaration(None=@n7.SubUV_Frame) {name=DP_SubUV_Frame_1, nodeColor=<rgba 1 0.1745 0.2654 1>}
  n8 = Multiply(A=ParticleColor().RGB, B=NamedRerouteUsage())
  n9 = VectorParameter() {group=Emissive, parameterName=Time of Day, channelNames {a=Night, b=Evening, g=Day, r=Morning}, defaultValue=<rgba 1 1 1 1>}
  NamedRerouteDeclaration(None=Multiply(
    A=MaterialFunctionCall(Emissive=@n8) {materialFunction=&MF_FX_EmissiveAdjustments}.Result
    B=MaterialFunctionCall(Value_Morning=@n9.Morning, Value_Day=@n9.Morning, Value_Evening=@n9.Morning, Value_Night=@n9.Morning) {materialFunction=&Scalar_Time_of_Day}.Result
   )) {name=Emissive Out_1, nodeColor=<rgba 1 0.3177 0.8839 1>}
  n10 = Multiply(A=ParticleColor().A, B=NamedRerouteUsage())
  NamedRerouteDeclaration(None=MaterialFunctionCall(Opacity In=@n10, DepthFade Distance=ScalarParameter() {defaultValue=100, group=Opacity, parameterName=Depth Fade Distance}) {materialFunction=&MF_VCVFX_DepthFade}.Result) {name=Opacity_1, nodeColor=<rgba 0.2324 0.9287 1 1>}
  NamedRerouteDeclaration(None=LinearInterpolate(
    A=@n10
    B=MaterialFunctionCall(Alpha Threshold=@n10) {materialFunction=&DitherTemporalAA}.Result
    Alpha=ScalarParameter() {group=Opacity, parameterName=Mask Dither}
   )) {name=Masked Opacity Out_1, nodeColor=<rgba 1 0 0.9609 1>}
  n11 = NamedRerouteUsage()
  n12 = AppendVector(A=@n11, B=NamedRerouteUsage())
  NamedRerouteDeclaration(None=LinearInterpolate(
    A=@n11
    B=ChannelMaskParameter(None=@n12) {group=Opacity, parameterName=Opacity Channel}.R
    Alpha=ScalarParameter() {group=Opacity, parameterName=Use Opacity Texture Channel?}
   )) {name=Opacity Texture, nodeColor=<rgba 0 1 0.2969 1>}
  NamedRerouteDeclaration(None=@n7.SubUV_ImagesY) {name=DP_SubUV_Images_Y, nodeColor=<rgba 1 0.1013 0.1453 1>}
  n13 = NamedRerouteUsage()
  n14 = ChannelMaskParameter(None=NamedRerouteDeclaration(None=@n6.RGB) {name=SubUVFunction RGB, nodeColor=<rgba 0 0.0782 1 1>}) {group=Texture, parameterName=Texture Channel}
  n15 = ScalarParameter() {group=Texture, parameterName=Use Texture Channel?}
  NamedRerouteDeclaration(None=Power(Base=LinearInterpolate(A=@n13, B=@n14.R, Alpha=@n15), Exp=ScalarParameter() {defaultValue=1, group=Texture, parameterName=Texture Power})) {name=Flipbook_Out_1, nodeColor=<rgba 0 1 0.1797 1>}
  NamedRerouteDeclaration(None=@n7.Packed Channel) {name=DP_Packed Channel_1, nodeColor=<rgba 1 0.0971 0.2313 1>}
  NamedRerouteDeclaration(None=@n7.SubUV_Images) {name=DP_SubUV Images_1, nodeColor=<rgba 1 0.1013 0.1453 1>}
  NamedRerouteUsage()
  NamedRerouteUsage()
  NamedRerouteUsage()
  Reroute(None=@n6.RGB)
  Reroute(None=@n6.Alpha)
 ]
}
