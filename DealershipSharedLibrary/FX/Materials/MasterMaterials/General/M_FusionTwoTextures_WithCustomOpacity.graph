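# Masked, unlit master FX material for Niagara sprites and ribbons that fuses two panning textures, tinted by particle color.
# Each texture has tiling, center scaling, optional distortion, channel selection, desaturation and contrast; the optional second blends by multiply or add.
# Opacity blends between the fused result and a custom, flippable texture, optionally multiplied by Texture 1 channels.
# Linear gradients fade front/back and optionally top/bottom edges, with depth fade skipped on mobile.
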Material /DealershipSharedLibrary/FX/Materials/MasterMaterials/General/M_FusionTwoTextures_WithCustomOpacity
graph {
 properties {
  bAutomaticallySetUsageInEditor=false
  bUsedWithNiagaraRibbons=true
  bUsedWithNiagaraSprites=true
  blendMode=BLEND_Masked
  shadingModel=MSM_Unlit
  usedShadingModels=Unlit
 }
 nodes [
  n1 = ParticleColor()
  n2 = Multiply(
   A=TextureCoordinate()
   B=MaterialFunctionCall(
    X=ScalarParameter() {defaultValue=1, group=Texture, parameterName=Tiling X, sortPriority=3}
    Y=ScalarParameter() {defaultValue=1, group=Texture, parameterName=Tiling Y, sortPriority=4}
   ) {materialFunction=&MakeFloat2}.Result
  )
  n3 = ScalarParameter() {defaultValue=1, group=Texture, parameterName=Texture Scale U}
  n4 = ScalarParameter() {defaultValue=1, group=Texture, parameterName=Texture Scale V}
  n5 = LinearInterpolate(
   A=@n2
   B=MaterialFunctionCall(UVs=@n2, Texture Scale=AppendVector(A=@n3, B=@n4)) {materialFunction=&ScaleUVsByCenter}.UVs
   Alpha=ScalarParameter() {group=Texture, parameterName=Scale UV By Center?, sliderMax=1}
  )
  n6 = ScalarParameter() {defaultValue=1, group=Distortion, parameterName=Tiling X Distortion, sortPriority=4}
  n7 = ScalarParameter() {defaultValue=1, group=Distortion, parameterName=Tiling Y Distortion, sortPriority=5}
  n8 = ScalarParameter() {defaultValue=1, group=Distortion, parameterName=Panner Speed X Distortion, sortPriority=6}
  n9 = ScalarParameter() {defaultValue=1, group=Distortion, parameterName=Panner Speed Y Distortion, sortPriority=7}
  n10 = VectorParameter() {group=Distortion, parameterName=RGB Channel Selector Distortion, sortPriority=1, defaultValue=<rgba 1 0 0 0>}
  n11 = StaticSwitchParameter(
   True=MaterialFunctionCall!(
    Texture=TextureObjectParameter() {group=Distortion, parameterName=Distortion Texture, sortPriority=0}
    Tiling=MaterialFunctionCall(X=@n6, Y=@n7) {materialFunction=&MakeFloat2}.Result
    Panner Time=ScalarParameter() {defaultValue=1, group=Distortion, parameterName=Panner Time Distortion, sortPriority=3}
    Offset XY
    Panner Speed=MaterialFunctionCall(X=@n8, Y=@n9) {materialFunction=&MakeFloat2}.Result
    Distortion Strength=ScalarParameter() {group=Distortion, parameterName=Distortion Strength, sortPriority=2}
    Channel Selector=@n10.RGB
   ) {materialFunction=/DealershipSharedLibrary/FX/Materials/MaterialFunctions/MF_TextureDistortion}.Distortion
   False=Constant()
  ) {group=_General, parameterName=Use Distortion 1?, sortPriority=3}
  n12 = ScalarParameter() {group=Texture, parameterName=Speed X Panner, sortPriority=5}
  n13 = ScalarParameter() {group=Texture, parameterName=Speed Y Panner, sortPriority=6}
  n14 = TextureSampleParameter2D(UVs=Panner(
    Coordinate=Add(A=@n5, B=@n11)
    Speed=MaterialFunctionCall(X=@n12, Y=@n13) {materialFunction=&MakeFloat2}.Result
   )) {group=Texture, parameterName=Texture 1}
  n15 = StaticComponentMaskParameter(None=@n14.RGB) {defaultB=true, defaultG=true, defaultR=true, group=Texture, parameterName=RGB Channel Selector, sortPriority=1}
  n16 = Add(A=StaticComponentMaskParameter(None=@n14.RGB) {defaultR=true, group=Texture, parameterName=Channel Add Layer, sortPriority=1}, B=@n15)
  n17 = ScalarParameter() {group=Texture, parameterName=Add Extra Channel Layer?, sliderMax=1}
  n18 = Power(
   Base=Desaturation(
    None=LinearInterpolate(A=@n15, B=@n16, Alpha=@n17)
    Fraction=ScalarParameter() {group=Texture, parameterName=Desaturation 0-1, sortPriority=2}
   )
   Exp=ScalarParameter() {defaultValue=1, group=Texture, parameterName=Contrast TEX 1}
  )
  n19 = Multiply(
   A=TextureCoordinate()
   B=MaterialFunctionCall(
    X=ScalarParameter() {defaultValue=1, group=Texture 2, parameterName=Tiling X TEX 2, sortPriority=3}
    Y=ScalarParameter() {defaultValue=1, group=Texture 2, parameterName=Tiling Y TEX 2, sortPriority=4}
   ) {materialFunction=&MakeFloat2}.Result
  )
  n20 = ScalarParameter() {defaultValue=1, group=Texture 2, parameterName=Texture 2 Scale U}
  n21 = ScalarParameter() {defaultValue=1, group=Texture 2, parameterName=Texture 2 Scale V}
  n22 = LinearInterpolate(
   A=@n19
   B=MaterialFunctionCall(UVs=@n19, Texture Scale=AppendVector(A=@n20, B=@n21)) {materialFunction=&ScaleUVsByCenter}.UVs
   Alpha=ScalarParameter() {group=Texture 2, parameterName=Scale UV TEX 2 By Center?, sliderMax=1}
  )
  n23 = ScalarParameter() {defaultValue=1, group=Distortion 2, parameterName=Tiling X Distortion 2, sortPriority=4}
  n24 = ScalarParameter() {defaultValue=1, group=Distortion 2, parameterName=Tiling Y Distortion 2, sortPriority=5}
  n25 = ScalarParameter() {defaultValue=1, group=Distortion, parameterName=Panner Speed X Distortion 2, sortPriority=6}
  n26 = ScalarParameter() {defaultValue=1, group=Distortion 2, parameterName=Panner Speed Y Distortion 2, sortPriority=7}
  n27 = VectorParameter() {group=Distortion 2, parameterName=RGB Channel Selector Distortion 2, sortPriority=1, defaultValue=<rgba 1 0 0 0>}
  n28 = StaticSwitchParameter(
   True=MaterialFunctionCall(
    Texture=TextureObjectParameter() {group=Distortion 2, parameterName=Distortion 2 Texture, sortPriority=0}
    Tiling=MaterialFunctionCall(X=@n23, Y=@n24) {materialFunction=&MakeFloat2}.Result
    Panner Time=ScalarParameter() {defaultValue=1, group=Distortion 2, parameterName=Panner Time Distortion 2, sortPriority=3}
    Panner Speed=MaterialFunctionCall(X=@n25, Y=@n26) {materialFunction=&MakeFloat2}.Result
    Distortion Strength=ScalarParameter() {group=Distortion 2, parameterName=Distortion Strength 2, sortPriority=2}
    Channel Selector=@n27.RGB
   ) {materialFunction=&MF_TextureDistortion}.Distortion
   False=Constant()
  ) {group=_General, parameterName=Use Distortion 2?, sortPriority=4}
  n29 = ScalarParameter() {group=Texture 2, parameterName=Speed X Panner TEX 2, sortPriority=5}
  n30 = ScalarParameter() {group=Texture 2, parameterName=Speed Y Panner TEX 2, sortPriority=6}
  n31 = TextureSampleParameter2D(UVs=Panner(
    Coordinate=Add(A=@n22, B=@n28)
    Speed=MaterialFunctionCall(X=@n29, Y=@n30) {materialFunction=&MakeFloat2}.Result
   )) {group=Texture 2, parameterName=Texture 2}
  n32 = Power(
   Base=Desaturation(
    None=StaticComponentMaskParameter(None=@n31.RGB) {defaultB=true, defaultG=true, defaultR=true, group=Texture 2, parameterName=RGB Channel Selector TEX 2}
    Fraction=ScalarParameter() {group=Texture 2, parameterName=Desaturation 0-1 TEX 2, sortPriority=2}
   )
   Exp=ScalarParameter() {defaultValue=1, group=Texture 2, parameterName=Contrast TEX 2}
  )
  n33 = VectorParameter() {group=Texture 2, parameterName=Custom Color TEX 2, defaultValue=<rgba 1 1 1 1>}
  n34 = Add(
   A=@n18
   B=LinearInterpolate(A=@n32, B=Multiply(A=@n33.RGB, B=@n32), Alpha=ScalarParameter() {group=Texture 2, parameterName=Use Custom Color?, sliderMax=1})
  )
  n35 = StaticSwitchParameter(
   True=LinearInterpolate(
    A=Multiply(A=@n18, B=@n32)
    B=Saturate(None=@n34)
    Alpha=ScalarParameter() {group=_General, parameterName="Fusion Mode 0=Multiply 1=Add", sortPriority=1}
   )
   False=@n18
  ) {group=_General, parameterName=Use Secondary Texture?, sortPriority=0}
  MaterialFunctionCall(Emissive=Multiply(A=@n1.RGB, B=@n35)) {materialFunction=&MF_FX_EmissiveAdjustments}
  n36 = ScalarParameter() {defaultValue=1, group=Opacity, parameterName=Custom Opacity U Scale}
  n37 = ScalarParameter() {defaultValue=1, group=Opacity, parameterName=Custom Opacity V Scale}
  n38 = TextureSampleParameter2D(UVs=MaterialFunctionCall(Texture Scale=AppendVector(A=@n36, B=@n37)) {materialFunction=&ScaleUVsByCenter}.UVs) {group=Opacity, parameterName=Custom Opacity, samplerSource=SSM_Clamp_WorldGroupSettings}
  n39 = Multiply(
   A=LinearInterpolate(A=@n38.RGB, B=OneMinus(None=@n38.RGB), Alpha=ScalarParameter() {group=Opacity, parameterName=Flip Custom Opacity?, sliderMax=1})
   B=ScalarParameter() {defaultValue=1, group=Opacity, parameterName=Custom Opacity Intensity}
  )
  n40 = Multiply(
   A=DotProduct(
    A=TextureSampleParameter2D(UVs=@n5) {group=Texture, parameterName=Texture 1}.RGB
    B=VectorParameter() {group=Opacity, parameterName=Texture Channel Multiplier, defaultValue=<rgba 1 0 0 1>}.RGB
   )
   B=ScalarParameter() {defaultValue=1, group=Opacity, parameterName=Texture Channel Intensity}
  ) {constB=8}
  n41 = LinearInterpolate(A=@n39, B=Multiply(A=Saturate(None=@n40), B=@n39), Alpha=ScalarParameter() {group=Opacity, parameterName=Multiply Opacity By TEX Channel?})
  NamedRerouteDeclaration(None=LinearInterpolate(
    A=Multiply(A=@n1.A, B=@n35)
    B=Multiply(A=@n41, B=@n1.A)
    Alpha=ScalarParameter() {group=_General, parameterName=Use Custom Opacity?, sliderMax=1}
   )) {name=Opacity, nodeColor=<rgba 1 1 1 1>}
  n42 = NamedRerouteUsage()
  n43 = Multiply(
   A=@n42
   B=DepthFade(FadeDistance=ScalarParameter() {defaultValue=100, group=Opacity, parameterName=DepthFade-Standard_Distance, sortPriority=0}) {fadeDistanceDefault=20}
  )
  n44 = Multiply(
   A=MaterialFunctionCall!(
    "Gradient U=0 V=1"
    "Invert gradient 0=False 1=True"
    "Contrast (Power)"=ScalarParameter() {defaultValue=1, group=Opacity, parameterName=Contrast Fade Front, sortPriority=5}
    "Brightness (Multiply)"=ScalarParameter() {defaultValue=1, group=Opacity, parameterName=Brightness Fade Front, sortPriority=6}
    "Dynamic Contrast (Power)"
    "Dynamic Brightness (Multiply)"
   ) {materialFunction=/DealershipSharedLibrary/FX/Materials/MaterialFunctions/MF_LinearGradientPlus}.Result
   B=MaterialFunctionCall(
    "Invert gradient 0=False 1=True"=Constant() {r=1}
    "Contrast (Power)"=ScalarParameter() {defaultValue=1, group=Opacity, parameterName=Contrast Fade Back, sortPriority=7}
    "Brightness (Multiply)"=ScalarParameter() {defaultValue=1, group=Opacity, parameterName=Brightness Fade Back, sortPriority=8}
   ) {materialFunction=&MF_LinearGradientPlus}.Result
  )
  n45 = MaterialFunctionCall(
   "Gradient U=0 V=1"=Constant() {r=1}
   "Contrast (Power)"=ScalarParameter() {defaultValue=1, group=Opacity, parameterName=Contrast Fade Top, sortPriority=9}
   "Brightness (Multiply)"=ScalarParameter() {defaultValue=1, group=Opacity, parameterName=Brightness Fade Top, sortPriority=10}
  ) {materialFunction=&MF_LinearGradientPlus}
  n46 = Constant() {r=1}
  n47 = MaterialFunctionCall(
   "Gradient U=0 V=1"=@n46
   "Invert gradient 0=False 1=True"=@n46
   "Contrast (Power)"=ScalarParameter() {defaultValue=1, group=Opacity, parameterName=Contrast Fade Bottom, sortPriority=11}
   "Brightness (Multiply)"=ScalarParameter() {defaultValue=1, group=Opacity, parameterName=Brightness Fade Bottom, sortPriority=12}
  ) {materialFunction=&MF_LinearGradientPlus}
  n48 = StaticSwitchParameter(True=Multiply(A=@n44, B=Multiply(A=@n45.Result, B=@n47.Result)), False=@n44) {group=Opacity, parameterName=Use Vertical Fade?, sortPriority=-1}
  Saturate(None=Multiply(A=ShadingPathSwitch(Default=@n43, Deferred=@n43, Mobile=@n42) {desc=Migrated from FeatureLevelSwitch}, B=@n48))
  Saturate(None=Multiply(A=@n42, B=@n48))
 ]
}
