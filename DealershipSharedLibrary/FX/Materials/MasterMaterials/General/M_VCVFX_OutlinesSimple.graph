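# Translucent, unlit, two-sided master material for Niagara sprite VFX, by name a simple outline effect.
# A channel-packed texture supplies two color masks and an opacity mask, with particle color and alpha available, feeding a two-term emissive with adjustable intensity.
# Optional noise erosion uses dynamic-parameter amount and softness, and a scalar switches color intensity to dynamic control.
# Shared functions handle emissive adjustments and depth-faded opacity.
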Material /DealershipSharedLibrary/FX/Materials/MasterMaterials/General/M_VCVFX_OutlinesSimple
graph {
 properties {
  bUsedWithNiagaraSprites=true
  blendMode=BLEND_Translucent
  pixelDepthOffsetMode=PDOM_Legacy
  shadingModel=MSM_Unlit
  twoSided=true
  usedShadingModels=Unlit
 }
 nodes [
  n1 = Multiply(A=NamedRerouteUsage(), B=NamedRerouteUsage())
  n2 = LinearInterpolate(
   A=ScalarParameter() {defaultValue=3, group=General, parameterName=B Color Intensity}
   B=DynamicParameter() {defaultValue=<rgba 0 0 0 0>, paramNames [Erosion Amount, Erosion Soft, B Color Intesity, Param4]}.B Color Intesity
   Alpha=ScalarParameter() {parameterName=UseDynamicParam}
  )
  n3 = NamedRerouteUsage()
  n4 = NamedRerouteUsage()
  MaterialFunctionCall(Emissive=Add(A=Multiply(A=@n1, B=@n2), B=Multiply(A=@n3, B=@n4))) {materialFunction=&MF_FX_EmissiveAdjustments}
  n5 = Multiply(A=NamedRerouteUsage(), B=NamedRerouteUsage())
  MaterialFunctionCall(Opacity In=Multiply(A=@n5, B=NamedRerouteUsage()), DepthFade Distance=ScalarParameter() {defaultValue=100, parameterName=DepthFade Distance}) {materialFunction=&MF_VCVFX_DepthFade}
  n6 = TextureSampleParameter2D() {
   desc="R= Color 1, G= Mask, B=Color 2"
   group=General
   parameterName=Main Texture
   texture=/DealershipSharedLibrary/FX/Textures/Flipbook/T_VCVFX_TrianglesSprites_RGB
  }
  NamedRerouteDeclaration(None=@n6.G) {name=Opacity Mask, nodeColor=<rgba 0 1 0.2266 1>}
  n7 = ParticleColor()
  NamedRerouteDeclaration(None=@n7.A) {name=Particle Alpha, nodeColor=<rgba 1 0 0.0937 1>}
  NamedRerouteDeclaration(None=@n7.RGB) {name=Particle RGB, nodeColor=<rgba 0 0.25 1 1>}
  NamedRerouteDeclaration(None=@n6.B) {name=Tex_Color 2, nodeColor=<rgba 0.5703 1 0 1>}
  NamedRerouteDeclaration(None=@n6.R) {name=Tex_Color 1, nodeColor=<rgba 1 0 0.8906 1>}
  n8 = TextureSampleParameter2D(UVs=Multiply(A=TextureCoordinate(), B=ScalarParameter() {defaultValue=1, parameterName=Erosion Scale UV})) {group=General, parameterName=Tex_Erosion, texture=/DealershipSharedLibrary/FX/Textures/Noise/T_VCVFX_CausticsNoise_A}
  n9 = DynamicParameter() {defaultValue=<rgba 0 0 0 0>, paramNames [Erosion Amount, Erosion Soft, B Color Intesity, Param4]}
  n10 = Divide(A=Subtract(A=Saturate(None=@n8.R), B=@n9.Erosion Amount) {constB=0.5}, B=@n9.Erosion Soft)
  n11 = ScalarParameter() {parameterName=UseDynamicParamErosion?}
  NamedRerouteDeclaration(None=LinearInterpolate(A=Saturate(None=@n10), Alpha=OneMinus(None=@n11))) {name=Erosion, nodeColor=<rgba 1 0.2109 0 1>}
  Reroute(None=Multiply(A=@n5, B=NamedRerouteUsage()))
 ]
}
