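# Material function for FX materials that outputs an animated world position offset vector, producing mesh waves along selectable axes.
# A sine of oriented, tiled UVs plus inverted, speed-scaled time is scaled by strength and a per-axis vector.
# A rounded scalar switches between those waves and a DynamicParameter vector input.
# An optional anchor mask multiplies the result by four MF_LinearGradientPlus fades (front, back, top, bottom), restricting motion to a UV region.
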MaterialFunction /DealershipSharedLibrary/FX/Materials/MaterialFunctions/VC_MF_WPO_MeshPerAxis
graph {nodes [
  n1 = TextureCoordinate()
  n2 = VectorParameter() {
   desc=You can modify the mask orientation base on your UVs
   group=WPO
   parameterName=OrientationWPO
   sortPriority=2
   channelNames {a=NA, b="", g="", r=""}
  }
  n3 = Time()
  n4 = Add(
   A=Divide(A=DotProduct(A=@n1, B=@n2.RGB), B=ScalarParameter() {defaultValue=1, group=WPO, parameterName=WPO_Tile, sortPriority=3})
   B=Multiply(A=ScalarParameter() {defaultValue=1, group=WPO, parameterName=WPO_Speed, sortPriority=4}, B=OneMinus(None=@n3))
  )
  n5 = VectorParameter() {
   desc=You can lock a specific axis to generate the WPO
   group=WPO
   parameterName=WPOPerAxis
   sortPriority=2
   channelNames {a=NA, b=Z, g=Y, r=X}
  }
  n6 = Multiply(A=Multiply(A=Sine(None=@n4), B=ScalarParameter() {group=WPO, parameterName=WPO_Strenght, sortPriority=3}), B=@n5.RGB)
  n7 = FunctionInput(Preview=Constant3Vector()) {bUsePreviewValueAsDefault=true, inputName=DynamicParameter}
  n8 = Round(None=ScalarParameter() {desc=To control WPO via dynamic parameters, group=WPO, parameterName=WPOWaves/DynamicVector, sortPriority=4})
  n9 = ScalarParameter() {defaultValue=1, group=WPO, parameterName=W_Contrast Fade Front, sortPriority=5}
  n10 = ScalarParameter() {defaultValue=1, group=WPO, parameterName=W_Brightness Fade Front, sortPriority=6}
  n11 = Constant() {r=1}
  n12 = ScalarParameter() {defaultValue=1, group=WPO, parameterName=W_Contrast Fade Back, sortPriority=7}
  n13 = ScalarParameter() {defaultValue=1, group=WPO, parameterName=W_Brightness Fade Back, sortPriority=8}
  n14 = Constant() {r=1}
  n15 = ScalarParameter() {defaultValue=1, group=WPO, parameterName=W_Contrast Fade Top, sortPriority=9}
  n16 = ScalarParameter() {defaultValue=1, group=WPO, parameterName=W_Brightness Fade Top, sortPriority=10}
  n17 = Constant() {r=1}
  n18 = ScalarParameter() {defaultValue=1, group=WPO, parameterName=W_Contrast Fade Bottom, sortPriority=11}
  n19 = ScalarParameter() {defaultValue=1, group=WPO, parameterName=W_Brightness Fade Bottom, sortPriority=12}
  n20 = Multiply(
   A=Multiply(
    A=MaterialFunctionCall!(
     "Gradient U=0 V=1"
     "Invert gradient 0=False 1=True"
     "Contrast (Power)"=@n9
     "Brightness (Multiply)"=@n10
     "Dynamic Contrast (Power)"
     "Dynamic Brightness (Multiply)"
    ) {materialFunction=&/MF_LinearGradientPlus}.Result
    B=MaterialFunctionCall("Invert gradient 0=False 1=True"=@n11, "Contrast (Power)"=@n12, "Brightness (Multiply)"=@n13) {materialFunction=&MF_LinearGradientPlus}.Result
   )
   B=Multiply(
    A=MaterialFunctionCall("Gradient U=0 V=1"=@n14, "Contrast (Power)"=@n15, "Brightness (Multiply)"=@n16) {materialFunction=&MF_LinearGradientPlus}.Result
    B=MaterialFunctionCall("Gradient U=0 V=1"=@n17, "Invert gradient 0=False 1=True"=@n17, "Contrast (Power)"=@n18, "Brightness (Multiply)"=@n19) {materialFunction=&MF_LinearGradientPlus}.Result
   )
  )
  n21 = ScalarParameter() {
   controlType=Enumeration
   desc=You can set up a mesh anchor to restrict the effect to a specific UV Region
   enumeration=/DealershipSharedLibrary/FX/Materials/Enums/Enum_DLR_TrueFalse
   group=WPO
   parameterName=Need__AnchorPivot?
   sortPriority=5
  }
  FunctionOutput(None=Multiply(A=LinearInterpolate(A=@n6, B=@n7, Alpha=@n8), B=LinearInterpolate(B=@n20, Alpha=@n21) {constA=1}))
 ]}
