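# Master unlit, masked, dithered VFX material for Niagara sprites, ribbons and meshes in the dealership shared library.
# A tiled, panning, optionally distorted texture and particle color drive emissive and opacity, with optional view-angle fake reflection, detail texture and Fresnel rim.
# A rotatable noise dissolve, optionally distorted, erodes opacity before depth fade.
# Dynamic parameter index 1 controls dissolve amount and scale; index 3 feeds switchable per-axis mesh WPO.
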Material /DealershipSharedLibrary/FX/Materials/MasterMaterials/General/M_VC_SimpleParticle
graph {
 properties {
  bAutomaticallySetUsageInEditor=false
  bUsedWithNiagaraMeshParticles=true
  bUsedWithNiagaraRibbons=true
  bUsedWithNiagaraSprites=true
  blendMode=BLEND_Masked
  ditherOpacityMask=true
  shadingModel=MSM_Unlit
  usedShadingModels=Unlit
 }
 nodes [
  LinearInterpolate(A=NamedRerouteUsage())
  n1 = ScalarParameter() {defaultValue=1, group=Texture, parameterName=Tiling X, sortPriority=3}
  n2 = ScalarParameter() {defaultValue=1, group=Texture, parameterName=Tiling Y, sortPriority=3}
  n3 = Add(
   A=Multiply(
    A=TextureCoordinate()
    B=MaterialFunctionCall(X=@n1, Y=@n2) {materialFunction=&MakeFloat2}.Result
   )
   B=MaterialFunctionCall(
    X=ScalarParameter() {group=Texture, parameterName=OfsetX, sortPriority=3}
    Y=ScalarParameter() {group=Texture, parameterName=OfsetY, sortPriority=4}
   ) {materialFunction=&MakeFloat2}.Result
  )
  n4 = ScalarParameter() {defaultValue=1, group=Distortion, parameterName=Tiling X Distortion, sortPriority=4}
  n5 = ScalarParameter() {defaultValue=1, group=Distortion, parameterName=Tiling Y Distortion, sortPriority=5}
  n6 = ScalarParameter() {defaultValue=1, group=Distortion, parameterName=Panner Speed X Distortion, sortPriority=6}
  n7 = ScalarParameter() {defaultValue=1, group=Distortion, parameterName=Panner Speed Y Distortion, sortPriority=7}
  n8 = VectorParameter() {group=Distortion, parameterName=RGB Channel Selector Distortion, sortPriority=1, defaultValue=<rgba 1 0 0 0>}
  n9 = StaticSwitchParameter(
   True=MaterialFunctionCall!(
    Texture=TextureObjectParameter() {group=Distortion, parameterName=Distortion Texture, sortPriority=0}
    Tiling=MaterialFunctionCall(X=@n4, Y=@n5) {materialFunction=&MakeFloat2}.Result
    Panner Time=ScalarParameter() {defaultValue=1, group=Distortion, parameterName=Panner Time Distortion, sortPriority=3}
    Offset XY
    Panner Speed=MaterialFunctionCall(X=@n6, Y=@n7) {materialFunction=&MakeFloat2}.Result
    Distortion Strength=ScalarParameter() {group=Distortion, parameterName=Distortion Strength, sortPriority=2}
    Channel Selector=@n8.RGB
   ) {materialFunction=/DealershipSharedLibrary/FX/Materials/MaterialFunctions/MF_TextureDistortion}.Distortion
   False=Constant()
  ) {group=_General, parameterName=Use Distortion?, sortPriority=2}
  n10 = ScalarParameter() {group=Texture, parameterName=Speed X Panner, sortPriority=5}
  n11 = ScalarParameter() {group=Texture, parameterName=Speed Y Panner, sortPriority=6}
  n12 = TextureSampleParameter2D(UVs=Panner(
    Coordinate=Add(A=@n3, B=@n9)
    Speed=MaterialFunctionCall(X=@n10, Y=@n11) {materialFunction=&MakeFloat2}.Result
   )) {group=Texture, parameterName=Texture, sortPriority=0, texture=/DealershipSharedLibrary/FX/Textures/Masks/T_VCVFX_Boost_2x2StarBeam_A}
  n13 = Desaturation(None=@n12.RGB, Fraction=Saturate(None=ScalarParameter() {defaultValue=1, group=Texture, parameterName=TextureEmissiveColorDesaturation}))
  n14 = ParticleColor()
  n15 = Multiply(A=@n14.RGB, B=ScalarParameter() {group=Reflection, parameterName=ColorSpecularIntensity}) {constB=15}
  n16 = Power(Base=@n13, Exp=ScalarParameter() {group=Reflection, parameterName=EmmisivePOWER}) {constExponent=6}
  n17 = StaticSwitchParameter(
   True=LinearInterpolate(
    A=Multiply(A=@n13, B=@n14.RGB)
    B=LinearInterpolate(A=@n14.RGB, B=@n15, Alpha=@n16)
    Alpha=ScalarParameter() {group=Reflection, parameterName=NeedReflection?, sortPriority=1}
   )
   False=@n14.RGB
  ) {group=_General, parameterName=Use Emmisive/Opacity Texture?, sortPriority=2}
  n18 = CameraVectorWS()
  n19 = Multiply(
   A=DotProduct(
    A=Transform(None=@n18) {transformSourceType=TRANSFORMSOURCE_World, transformType=TRANSFORM_Tangent}
    B=Constant3Vector() {constant=<rgba 0 0 1 0>}
   )
   B=ScalarParameter() {defaultValue=1, group=Reflection, parameterName=StepForce, sortPriority=7}
  ) {constB=3}
  n20 = ScalarParameter() {group=Reflection, parameterName=StepAmount, sortPriority=6}
  MaterialExpressionConstantClamp_13 = Saturate(None=SmoothStep(
    Min=ScalarParameter() {group=Reflection, parameterName=SmoothReflectionMin, sortPriority=4}
    Max=ScalarParameter() {defaultValue=1, group=Reflection, parameterName=SmoothReflextionMax, sortPriority=5}
    Value=Subtract(A=@n19, B=@n20)
   ))
  n21 = LinearInterpolate(
   A=@n17
   B=Multiply(A=@MaterialExpressionConstantClamp_13, B=@n17)
   Alpha=ScalarParameter() {group=Reflection, parameterName=ReflectionAmount, sortPriority=1}
  )
  n22 = Multiply(A=@n21, B=TextureSampleParameter2D() {group=Reflection, parameterName=Detailed_Texture, sortPriority=9}.RGB)
  n23 = ScalarParameter() {group=Reflection, parameterName=DetailedTextureInfluence, sortPriority=8}
  n24 = Round(None=ScalarParameter() {group=Reflection, parameterName=NeedReflection?, sortPriority=1})
  n25 = Multiply(
   A=LinearInterpolate(A=@n17, B=LinearInterpolate(A=@n21, B=@n22, Alpha=@n23), Alpha=Saturate(None=@n24))
   B=ScalarParameter() {defaultValue=1, group=_General, parameterName=Emissive Gain}
  )
  n26 = VectorParameter() {group=Fresnel, parameterName=CustomColorFresnel, defaultValue=<rgba 1 1 1 1>}
  n27 = Multiply(A=NamedRerouteUsage(), B=ScalarParameter() {group=Fresnel, parameterName=Fresnel Amount, sortPriority=0})
  MaterialFunctionCall(Emissive=Add(A=NamedRerouteDeclaration(None=@n25) {name=Color_1, nodeColor=<rgba 1 0 0.0703 1>}, B=Multiply(A=@n26.RGB, B=@n27))) {materialFunction=&MF_FX_EmissiveAdjustments}
  Multiply(
   A=LinearInterpolate(
    A=VectorParameter() {group=_General, parameterName=Base Color, defaultValue=<rgba 0 0 0 1>}.RGB
    B=ParticleColor().RGB
    Alpha=ScalarParameter() {group=_General, parameterName=Use Particle Color as Base Color?}
   )
   B=ScalarParameter() {defaultValue=1, group=_General, parameterName=Base Color Gain}
  )
  n28 = ScalarParameter() {
   controlType=Enumeration
   enumeration=/DealershipSharedLibrary/FX/Materials/Enums/Enum_DLR_TrueFalse
   group=WPO
   parameterName=UseWPO?
   sliderMax=1
   sortPriority=0
  }
  n29 = DynamicParameter() {parameterIndex=3, defaultValue=<rgba 0 0 0 0>, paramNames [WPO X, WPO Y, WPO Z, Param4]}
  NamedRerouteDeclaration(None=If(
    A=@n28
    "A > B"=MaterialFunctionCall!(DynamicParameter=@n29.RGB) {materialFunction=/DealershipSharedLibrary/FX/Materials/MaterialFunctions/VC_MF_WPO_MeshPerAxis}.Result
    "A < B"=Constant()
   ) {constB=1}) {name=WPO, nodeColor=<rgba 0.6875 1 0 1>}
  n30 = Fresnel(
   ExponentIn=ScalarParameter() {defaultValue=5, group=Fresnel, parameterName=Fresnel - Exponent, sortPriority=3}
   BaseReflectFractionIn=ScalarParameter() {defaultValue=0.04, group=Fresnel, parameterName=Fresnel - BaseReflectFraction, sortPriority=3}
  )
  n31 = ScalarParameter() {group=Fresnel, parameterName=Fresnel - Bias, sortPriority=1}
  n32 = Multiply(
   A=Power(Base=@n30, Exp=ScalarParameter() {defaultValue=7, group=Fresnel, parameterName=Edge Highlight Thickness, sortPriority=2}) {constExponent=10}
   B=ScalarParameter() {defaultValue=50, group=Fresnel, parameterName=Edge Highlight Amount, sortPriority=2}
  ) {constB=50}
  NamedRerouteDeclaration(None=Add(A=Add(A=@n30, B=@n31), B=Saturate(None=@n32))) {name=Fresnel, nodeColor=<rgba 0.1981 1 0.1878 1>}
  NamedRerouteUsage()
  n33 = Multiply(A=@n26.A, B=@n27)
  n34 = Desaturation(
   None=DotProduct(A=@n12.RGB, B=VectorParameter() {group=Texture, parameterName=RGB Channel Selector, sortPriority=1, defaultValue=<rgba 1 0 0 0>}.RGB)
   Fraction=ScalarParameter() {group=Texture, parameterName=Desaturation 0-1, sortPriority=2}
  )
  n35 = StaticSwitchParameter(
   True=@n12.A
   False=StaticSwitchParameter(True=Multiply(A=@n14.A, B=@n34), False=@n14.A) {group=_General, parameterName=Use Texture?, sortPriority=1}
  ) {group=_General, parameterName=Use Emmisive/Opacity Texture?, sortPriority=2}
  n36 = Multiply(A=@n9, B=ScalarParameter() {parameterName=UseDistortionToDissolve})
  n37 = TextureCoordinate()
  n38 = AppendVector(
   A=ScalarParameter() {defaultValue=1, group=Dissolve, parameterName=DissolveTiling X, sortPriority=2}
   B=ScalarParameter() {defaultValue=1, group=Dissolve, parameterName=DissolveTiling Y, sortPriority=3}
  )
  n39 = DynamicParameter() {parameterIndex=1, defaultValue=<rgba 0 0 0 1>, paramNames [Dissolve, Dissolve Scale, Param3, Param4]}
  n40 = ScalarParameter() {group=Dissolve, parameterName=Scroll X, sortPriority=4}
  n41 = ScalarParameter() {group=Dissolve, parameterName=Scroll Y, sortPriority=5}
  n42 = Add(A=Multiply(A=@n37, B=Multiply(A=@n38, B=@n39.Dissolve Scale)), B=Panner(Coordinate=@n37, Speed=AppendVector(A=@n40, B=@n41)))
  n43 = TextureSampleParameter2D(UVs=MaterialFunctionCall(UVs=Add(A=@n36, B=@n42), "Rotation Angle (0-1)"=ScalarParameter() {group=Dissolve, parameterName=Dissolve - Rotation}) {materialFunction=&CustomRotator}.Rotated Values) {
   group=Dissolve
   parameterName=Dissolve Texture
   samplerType=SAMPLERTYPE_LinearColor
   sortPriority=1
   texture=/DealershipSharedLibrary/FX/Textures/Noise/T_VCVFX_Noise
  }
  n44 = LinearInterpolate(
   A=@n43.RGB
   B=ChannelMaskParameter(None=@n43.RGB) {group=Dissolve, parameterName=Dissolve - Texture Channel}.R
   Alpha=ScalarParameter() {group=Dissolve, parameterName=Dissolve - Use Texture Channel?}
  )
  n45 = DynamicParameter() {parameterIndex=1, defaultValue=<rgba 0 0 0 1>, paramNames [Dissolve, Dissolve Scale, Param3, Param4]}
  n46 = Saturate(None=Divide(A=Subtract(A=@n44, B=@n45.Dissolve), B=ScalarParameter() {defaultValue=1, group=Dissolve, parameterName=Dissolve_Softness}) {constB=0.2})
  n47 = ScalarParameter() {group=Dissolve, parameterName=UseDissolve, sortPriority=0}
  n48 = Saturate(None=LinearInterpolate(A=@n35, B=Multiply(A=@n35, B=@n46), Alpha=Saturate(None=@n47)))
  Reroute(None=MaterialFunctionCall(Opacity In=Add(A=@n33, B=@n48), DepthFade Distance=ScalarParameter() {group=_General, parameterName=Depth Fade Distance}) {materialFunction=&MF_VCVFX_DepthFade}.Result)
 ]
}
